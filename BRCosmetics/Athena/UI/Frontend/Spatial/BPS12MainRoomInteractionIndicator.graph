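# Frontend spatial navigation indicator widget (FortEventLevelNavigationWidget child) named for the BPS12 main room, marking an interactable object.
# An overlay stacks center and outer material icons beside a clipped, initially hidden border holding a face-button action prompt and "Select" label.
# Object hover begin and end play the OnHover animation forward and in reverse, popping and rotating the icons, recoloring the center material and sliding the prompt open.
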WidgetBlueprint /BRCosmetics/Athena/UI/Frontend/Spatial/BPS12MainRoomInteractionIndicator
animations [OnHover {
  binding_count=4
  length_seconds=0.25
  object=&self:OnHover
  sequence {bindings [
    Overlay_10 {
     id=#1
     possessed_class=Overlay
     tracks [
      {
       type=MovieScene2DTransformTrack
       sections [{
         type=MovieScene2DTransformSection
         channels [
          Translation.X {type=MovieSceneScriptingFloatChannel}
          Translation.Y {type=MovieSceneScriptingFloatChannel}
          Angle {type=MovieSceneScriptingFloatChannel}
          Scale.X {type=MovieSceneScriptingFloatChannel, keys [[0,0.75],[2,1.5],[4,1.25]]}
          Scale.Y {type=MovieSceneScriptingFloatChannel, keys [[0,0.75],[2,1.5],[4,1.25]]}
          Shear.X {type=MovieSceneScriptingFloatChannel}
          Shear.Y {type=MovieSceneScriptingFloatChannel}
         ]
        }]
      }
      {
       type=MovieSceneFloatTrack
       display_name=Render Opacity
       sections [{type=MovieSceneFloatSection, channels [None {type=MovieSceneScriptingFloatChannel, keys [[0,0.5],[2,1]]}]}]
      }
     ]
    }
    Border_0 {
     id=#2
     possessed_class=Border
     tracks [
      {
       type=MovieSceneFloatVectorTrack
       display_name=Desired Size Scale
       sections [{
         type=MovieSceneFloatVectorSection
         channels [
          Vector.X {type=MovieSceneScriptingFloatChannel, keys [[2.0003,0],[5,1]]}
          Vector.Y {type=MovieSceneScriptingFloatChannel, const [1,1]}
         ]
        }]
      }
      {
       type=MovieSceneFloatTrack
       display_name=Render Opacity
       sections [{type=MovieSceneFloatSection, channels [None {type=MovieSceneScriptingFloatChannel, keys [[2,0],[5,1]]}]}]
      }
      {
       type=MovieScene2DTransformTrack
       sections [{
         type=MovieScene2DTransformSection
         channels [
          Translation.X {type=MovieSceneScriptingFloatChannel, keys [[2,-32],[5,0]]}
          Translation.Y {type=MovieSceneScriptingFloatChannel}
          Angle {type=MovieSceneScriptingFloatChannel}
          Scale.X {type=MovieSceneScriptingFloatChannel}
          Scale.Y {type=MovieSceneScriptingFloatChannel}
          Shear.X {type=MovieSceneScriptingFloatChannel}
          Shear.Y {type=MovieSceneScriptingFloatChannel}
         ]
        }]
      }
     ]
    }
    Image {
     id=#3
     possessed_class=Image
     tracks [
      {
       type=MovieScene2DTransformTrack
       sections [{
         type=MovieScene2DTransformSection
         channels [
          Translation.X {type=MovieSceneScriptingFloatChannel}
          Translation.Y {type=MovieSceneScriptingFloatChannel}
          Angle {type=MovieSceneScriptingFloatChannel, keys [[0,90],[5,0]]}
          Scale.X {type=MovieSceneScriptingFloatChannel}
          Scale.Y {type=MovieSceneScriptingFloatChannel}
          Shear.X {type=MovieSceneScriptingFloatChannel}
          Shear.Y {type=MovieSceneScriptingFloatChannel}
         ]
        }]
      }
      {
       type=MovieSceneWidgetMaterialTrack
       sections [{
         type=MovieSceneParameterSection
         channels [
          ColorR {type=MovieSceneScriptingFloatChannel, keys [[0,1],[5,0]]}
          ColorG {type=MovieSceneScriptingFloatChannel, keys [[0,1],[5,0.4709]]}
          ColorB {type=MovieSceneScriptingFloatChannel, const [1,2]}
          ColorA {type=MovieSceneScriptingFloatChannel, const [1,2]}
         ]
        }]
      }
     ]
    }
    Image_16 {
     id=#4
     possessed_class=Image
     tracks [
      {
       type=MovieSceneFloatTrack
       display_name=Render Opacity
       sections [{type=MovieSceneFloatSection, channels [None {type=MovieSceneScriptingFloatChannel, keys [[0,0],[5,1]]}]}]
      }
      {
       type=MovieScene2DTransformTrack
       sections [{
         type=MovieScene2DTransformSection
         channels [
          Translation.X {type=MovieSceneScriptingFloatChannel}
          Translation.Y {type=MovieSceneScriptingFloatChannel}
          Angle {type=MovieSceneScriptingFloatChannel, keys [[0,-120],[5,0]]}
          Scale.X {type=MovieSceneScriptingFloatChannel}
          Scale.Y {type=MovieSceneScriptingFloatChannel}
          Shear.X {type=MovieSceneScriptingFloatChannel}
          Shear.Y {type=MovieSceneScriptingFloatChannel}
         ]
        }]
      }
     ]
    }
   ]}
 }]
graphs [EventGraph {nodes [
   Event OnObjectHoverBegin!() -> (OutputDelegate, then, NavWidgetSettings:"Nav Widget Settings Structure (by ref)") {is=title K2Node_Event}
   > PlayAnimationForward(InAnimation=Get OnHover() -> (OnHover:Widget Animation Object Reference))
   Event OnObjectHoverEnd() {is=title K2Node_Event}
   > PlayAnimationReverse(InAnimation=Get OnHover())
   Event PreConstruct()
   Event Construct()
   Event Tick()
  ]}]
variables []
widgets {
 info {
  inheritedWidgetCount=1
  parentClass=/Script/FortniteGame.FortEventLevelNavigationWidget
  rootWidgetClass=&HorizontalBox
  widgetCount=9
 }
 widgets [
  HorizontalBox_2 {widgetClassPath=&HorizontalBox, widget=HorizontalBox_2}
  Overlay_10 {
   widgetClassPath=&Overlay
   parent=HorizontalBox_2
   widget=Overlay_10
   properties {renderOpacity=0.5, renderTransform {angle=0, scale=<1.25 1.25>, shear=<2 -4>, translation=<0 0>}}
   slot {class=HorizontalBoxSlot}
  }
  Image_16 {
   widgetClassPath=&Image
   parent=Overlay_10
   widget=Image_16
   properties {
    renderOpacity=0
    visibility=HitTestInvisible
    brush {
     bIsDynamicallyLoaded=false
     drawAs=Image
     imageType=NoImage
     mirroring=NoMirror
     resourceName=None
     resourceObject=/BRCosmetics/Athena/UI/Frontend/Spatial/Art/M_UI_BPS12_InteractionIndicator_Outer
     tiling=NoTile
     imageSize=<40 40>
     margin=<ltrb 0 0 0 0>
     outlineSettings {
      bUseBrushTransparency=false
      roundingType=HalfHeightRadius
      width=0
      color {colorUseRule=UseColor_Specified, specifiedColor=<rgba 0 0 0 0>}
      cornerRadii=<0 0 0 0>
     }
     tintColor {colorUseRule=UseColor_Specified, specifiedColor=<rgba 1 1 1 1>}
     uVRegion {bIsValid=false, max=<0 0>, min=<0 0>}
    }
   }
   slot {class=OverlaySlot, properties {horizontalAlignment=HAlign_Center, verticalAlignment=VAlign_Center}}
  }
  Image {
   widgetClassPath=&Image
   parent=Overlay_10
   widget=Image
   properties {
    visibility=HitTestInvisible
    brush {
     bIsDynamicallyLoaded=false
     drawAs=Image
     imageType=NoImage
     mirroring=NoMirror
     resourceName=None
     resourceObject=/BRCosmetics/Athena/UI/Frontend/Spatial/Art/M_UI_BPS12_InteractionIndicator_Center
     tiling=NoTile
     imageSize=<40 40>
     margin=<ltrb 0 0 0 0>
     outlineSettings {
      bUseBrushTransparency=false
      roundingType=HalfHeightRadius
      width=0
      color {colorUseRule=UseColor_Specified, specifiedColor=<rgba 0 0 0 0>}
      cornerRadii=<0 0 0 0>
     }
     tintColor {colorUseRule=UseColor_Specified, specifiedColor=<rgba 1 1 1 1>}
     uVRegion {bIsValid=false, max=<0 0>, min=<0 0>}
    }
    renderTransform {angle=90, scale=<1 1>, shear=<0 0>, translation=<0 0>}
   }
   slot {class=OverlaySlot, properties {horizontalAlignment=HAlign_Center, verticalAlignment=VAlign_Center}}
  }
  Border_0 {
   widgetClassPath=&Border
   parent=HorizontalBox_2
   widget=Border_0
   properties {
    clipping=ClipToBounds
    renderOpacity=0
    background {
     bIsDynamicallyLoaded=false
     drawAs=NoDrawType
     imageType=NoImage
     mirroring=NoMirror
     resourceName=None
     resourceObject=null
     tiling=NoTile
     imageSize=<32 32>
     margin=<ltrb 0 0 0 0>
     outlineSettings {
      bUseBrushTransparency=false
      roundingType=HalfHeightRadius
      width=0
      color {colorUseRule=UseColor_Specified, specifiedColor=<rgba 0 0 0 0>}
      cornerRadii=<0 0 0 0>
     }
     tintColor {colorUseRule=UseColor_Specified, specifiedColor=<rgba 1 1 1 1>}
     uVRegion {bIsValid=false, max=<0 0>, min=<0 0>}
    }
    desiredSizeScale=<0 1>
    padding=<ltrb 4 0 0 0>
    renderTransform {angle=0, scale=<0.875 1>, shear=<4 2>, translation=<0 2>}
   }
   slot {class=HorizontalBoxSlot, properties {padding=<ltrb -4 0 0 0>, size {sizeRule=Fill, value=1}}}
  }
  HorizontalBox_0 {
   widgetClassPath=&HorizontalBox
   parent=Border_0
   widget=HorizontalBox_0
   slot {class=BorderSlot, properties {padding=<ltrb 4 0 0 0>}}
  }
  InputActionWidget {
   widgetClassPath=&CommonActionWidget
   bIsVariable=true
   parent=HorizontalBox_0
   widget=InputActionWidget
   properties {inputActions [{dataTable=&InputActionDataTable, rowName=GenericFaceButtonBottom}]}
   slot {class=HorizontalBoxSlot, properties {horizontalAlignment=HAlign_Center, verticalAlignment=VAlign_Center}}
  }
  CommonTextBlock_41 {
   widgetClassPath=&CommonTextBlock
   parent=HorizontalBox_0
   widget=CommonTextBlock_41
   properties {
    text=Select
    textTransformPolicy=ToUpper
    font {
     bForceMonospaced=false
     bMaterialIsStencil=false
     fontMaterial=null
     fontObject=&BurbankSmall
     letterSpacing=0
     monospacedWidth=1
     size=18
     skewAmount=0
     typefaceFontName=Black
     outlineSettings {
      bApplyOutlineToDropShadows=false
      bMiteredCorners=false
      bSeparateFillAlpha=false
      outlineMaterial=null
      outlineSize=0
      outlineColor=<rgba 0 0 0 1>
     }
    }
    shadowOffset=<0 0>
   }
   slot {class=HorizontalBoxSlot, properties {verticalAlignment=VAlign_Center, padding=<ltrb 8 0 0 0>, size {sizeRule=Fill, value=1}}}
  }
  Button_Navigation {widgetClassPath=&CommonButtonLegacy, bInherited=true, widget=null}
 ]
}
