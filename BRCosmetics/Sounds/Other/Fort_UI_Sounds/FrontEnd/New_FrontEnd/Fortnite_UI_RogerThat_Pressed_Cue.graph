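# Sound cue for pressing the Roger That button in the Fortnite front-end UI.
# A mixer layers the standard main button click with a Season 17 battle pass Roger That stinger, mixing the stinger louder.
# The stinger runs through a modulator that pins it to a slightly lowered, fixed pitch.
# It plays on the UI sound class, with concurrency capped at one instance that stops the oldest.
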SoundCue /BRCosmetics/Sounds/Other/Fort_UI_Sounds/FrontEnd/New_FrontEnd/Fortnite_UI_RogerThat_Pressed_Cue
root=Mixer_0
nodes [
 {id=Mixer_0, children [WavePlayer_1,Modulator_0], properties {inputVolume [1,1.25]}}
 {id=Modulator_0, children [WavePlayer_0], properties {pitchMax=0.98, pitchMin=0.98, volumeMax=1, volumeMin=1}}
 {id=WavePlayer_0, properties {soundWaveAssetPtr=/BRCosmetics/Sounds/Other/Fort_UI_Sounds/FrontEnd/BattlePass/S17_BP_Roger_That_Stinger}}
 {id=WavePlayer_1, properties {soundWaveAssetPtr=/Game/Sounds/Fort_UI_Sounds/FrontEnd/New_FrontEnd/UI_FN_Button_Click_Main_01}}
]
properties {
 maxDistance=2097152
 soundClassObject=&Fort_UI
 volumeMultiplier=1
 concurrencyOverrides {maxCount=1, resolutionRule=StopOldest}
}
