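# UI sound cue for the front-end "About" modal pop-up, played in the UI sound class.
# A mixer layers a challenge-grant popup wave at full level with a Season 17 Battle Pass About-modal stinger at slightly reduced level.
# Concurrency allows only one instance at a time and blocks new plays while it is active.
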SoundCue /BRCosmetics/Sounds/Other/Fort_UI_Sounds/FrontEnd/UI_S14_AboutModal_PopUp_Cue
root=Mixer_0
nodes [
 {id=Mixer_0, children [WavePlayer_1,WavePlayer_0], properties {inputVolume [1,0.75]}}
 {id=WavePlayer_0, properties {soundWaveAssetPtr=/BRCosmetics/Sounds/Other/Fort_UI_Sounds/FrontEnd/BattlePass/S17_BP_AboutModalPopUp_Stinger}}
 {id=WavePlayer_1, properties {soundWaveAssetPtr=&/GrantChallenge_Popup_01}}
]
properties {
 maxDistance=2097152
 soundClassObject=&Fort_UI
 volumeMultiplier=1
 concurrencyOverrides {maxCount=1, resolutionRule=PreventNew}
}
