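# Bullet-style row widget for the BPS12 About modal, pairing a small skewed image marker with sheared, auto-wrapping rich text.
# PreConstruct and Construct both set the rich text from the text variable, so the designer preview matches runtime.
# A GoldnBlack flag selects gold or cyan for text and marker, but that colour chain has no incoming exec connection, so the widget keeps its default cyan.
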WidgetBlueprint /FortniteCoreUI/Frontend/Spatial/BPS12_AboutModal_Entry
graphs [EventGraph {nodes [
   n1 = Get GoldnBlack() -> (GoldnBlack:bool)
   SetDefaultColorAndOpacity!(
    execute
    self:Rich Text Block Object Reference=Get CommonRichTextBlock_96() -> (CommonRichTextBlock_96:Common Rich Text Block Object Reference)
    InColorAndOpacity:Slate Color Structure=Select(
     Option 0:Slate Color Structure=Make Slate Color(SpecifiedColor=(R=0.0331,G=0.855,B=1,A=1))
     Option 1:Slate Color Structure=Make Slate Color(SpecifiedColor=(R=0.7157,G=0.4793,B=0.0742,A=1))
     Index=@n1
    ) -> (ReturnValue:Slate Color Structure)
   ) -> (then)
   > SetColorAndOpacity(
    self:Image Object Reference=Get Image_27() -> (Image_27:Image Object Reference)
    InColorAndOpacity:Linear Color Structure=Select(Option 0:Linear Color Structure=(R=0.0331,G=0.855,B=1,A=1), Option 1:Linear Color Structure=(R=0.7157,G=0.4793,B=0.0742,A=1), Index=@n1) -> (ReturnValue:Linear Color Structure)
   )
   n2 = Event PreConstruct()
   SetText(execute=[@n2.then, @n3.then], self=Get CommonRichTextBlock_96(), InText=Get text() -> (text:Text))
   n3 = Event Construct()
  ]}]
variables [
 GoldnBlack {type=boolean}
 text {type=string, default="This is the <text color=\"FFFFFFFF\" fontface=\"black\">description</>"}
]
widgets {
 info {inheritedWidgetCount=0, parentClass=&UserWidget, rootWidgetClass=&HorizontalBox, widgetCount=3}
 widgets [
  HorizontalBox_14 {widgetClassPath=&HorizontalBox, widget=HorizontalBox_14, properties {flowDirectionPreference=Culture}}
  Image_27 {
   widgetClassPath=&Image
   bIsVariable=true
   parent=HorizontalBox_14
   widget=Image_27
   properties {
    brush {
     bIsDynamicallyLoaded=false
     drawAs=Image
     imageType=NoImage
     mirroring=NoMirror
     resourceName=None
     resourceObject=/Game/Athena/UI/ReplayBrowser/M_Ui_Skewed_Vfor9slices_Inst_Inst
     tiling=NoTile
     imageSize=<20 20>
     margin=<ltrb 0 0 0 0>
     outlineSettings {
      bUseBrushTransparency=false
      roundingType=HalfHeightRadius
      width=0
      color {colorUseRule=UseColor_Specified, specifiedColor=<rgba 0 0 0 0>}
      cornerRadii=<0 0 0 0>
     }
     tintColor {colorUseRule=UseColor_Specified, specifiedColor=<rgba 1 1 1 1>}
     uVRegion {bIsValid=false, max=<0 0>, min=<0 0>}
    }
    colorAndOpacity=<rgba 0.0331 0.855 1 1>
   }
   slot {class=HorizontalBoxSlot, properties {horizontalAlignment=HAlign_Left, verticalAlignment=VAlign_Top, padding=<ltrb 0 10 12 0>}}
  }
  CommonRichTextBlock_96 {
   widgetClassPath=&CommonRichTextBlock
   bIsVariable=true
   parent=HorizontalBox_14
   widget=CommonRichTextBlock_96
   properties {
    autoWrapText=true
    bOverrideDefaultStyle=true
    lineHeightPercentage=0.9
    text="This is the <text color=\"FFFFFFFF\" fontface=\"black\">description</>"
    defaultTextStyleOverride {
     overflowPolicy=Clip
     transformPolicy=None
     colorAndOpacity {colorUseRule=UseColor_Specified, specifiedColor=<rgba 0.0331 0.855 1 1>}
     font {
      bForceMonospaced=false
      bMaterialIsStencil=false
      fontMaterial=null
      fontObject=&BurbankSmall
      letterSpacing=0
      monospacedWidth=1
      size=24
      skewAmount=0
      typefaceFontName=Bold
      outlineSettings {
       bApplyOutlineToDropShadows=false
       bMiteredCorners=false
       bSeparateFillAlpha=false
       outlineMaterial=null
       outlineSize=0
       outlineColor=<rgba 0 0 0 1>
      }
     }
     highlightColor {colorUseRule=UseColor_Specified, specifiedColor=<rgba 0 0 0 1>}
     highlightShape {
      bIsDynamicallyLoaded=false
      drawAs=Image
      imageType=NoImage
      mirroring=NoMirror
      resourceName=None
      resourceObject=null
      tiling=NoTile
      imageSize=<32 32>
      margin=<ltrb 0 0 0 0>
      outlineSettings {
       bUseBrushTransparency=false
       roundingType=HalfHeightRadius
       width=0
       color {colorUseRule=UseColor_Specified, specifiedColor=<rgba 0 0 0 0>}
       cornerRadii=<0 0 0 0>
      }
      tintColor {colorUseRule=UseColor_Specified, specifiedColor=<rgba 1 1 1 1>}
      uVRegion {bIsValid=false, max=<0 0>, min=<0 0>}
     }
     selectedBackgroundColor {colorUseRule=UseColor_Foreground, specifiedColor=<rgba 1 0 1 1>}
     shadowColorAndOpacity=<rgba 0 0 0 1>
     shadowOffset=<0 0>
     strikeBrush {
      bIsDynamicallyLoaded=false
      drawAs=Image
      imageType=NoImage
      mirroring=NoMirror
      resourceName=None
      resourceObject=null
      tiling=NoTile
      imageSize=<32 32>
      margin=<ltrb 0 0 0 0>
      outlineSettings {
       bUseBrushTransparency=false
       roundingType=HalfHeightRadius
       width=0
       color {colorUseRule=UseColor_Specified, specifiedColor=<rgba 0 0 0 0>}
       cornerRadii=<0 0 0 0>
      }
      tintColor {colorUseRule=UseColor_Specified, specifiedColor=<rgba 1 1 1 1>}
      uVRegion {bIsValid=false, max=<0 0>, min=<0 0>}
     }
     underlineBrush {
      bIsDynamicallyLoaded=false
      drawAs=Image
      imageType=NoImage
      mirroring=NoMirror
      resourceName=None
      resourceObject=null
      tiling=NoTile
      imageSize=<32 32>
      margin=<ltrb 0 0 0 0>
      outlineSettings {
       bUseBrushTransparency=false
       roundingType=HalfHeightRadius
       width=0
       color {colorUseRule=UseColor_Specified, specifiedColor=<rgba 0 0 0 0>}
       cornerRadii=<0 0 0 0>
      }
      tintColor {colorUseRule=UseColor_Specified, specifiedColor=<rgba 1 1 1 1>}
      uVRegion {bIsValid=false, max=<0 0>, min=<0 0>}
     }
    }
    margin=<ltrb 0 0 0 8>
    renderTransform {angle=0, scale=<1 1>, shear=<-8 0>, translation=<0 0>}
   }
   slot {class=HorizontalBoxSlot, properties {verticalAlignment=VAlign_Top, padding=<ltrb 0 0 0 20>, size {sizeRule=Fill, value=1}}}
  }
 ]
}
