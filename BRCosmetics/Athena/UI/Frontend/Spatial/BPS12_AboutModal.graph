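# Frontend "About" info modal (AthenaAboutModal child) showing a title, configurable header, scrollable text entries and an item image over a blurred background.
# PreConstruct and OnSetContent rebuild entry widgets and apply header settings; PreConstruct also sets height and rebuilds the scrollbar style for a gold or cyan thumb.
# Activation plays the ShowUp animation, with a pop-up sound kept for later fade-out.
# Accept or touch-only close button broadcasts CloseAbout and fades that sound.
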WidgetBlueprint /BRCosmetics/Athena/UI/Frontend/Spatial/BPS12_AboutModal
animations [ShowUp {
  binding_count=5
  length_seconds=1
  object=&self:ShowUp
  sequence {bindings [
    ModalHeader {
     id=#1
     possessed_class=HorizontalBox
     tracks [
      {
       type=MovieSceneFloatTrack
       display_name=Render Opacity
       sections [{type=MovieSceneFloatSection, end_frame=20, start_frame=0, channels [None {type=MovieSceneScriptingFloatChannel, keys [[5,0],[12,1]]}]}]
      }
      {
       type=MovieScene2DTransformTrack
       sections [{
         type=MovieScene2DTransformSection
         end_frame=20
         start_frame=0
         channels [Translation.X {type=MovieSceneScriptingFloatChannel, keys [[5,-256],[10,0]]}]
        }]
      }
     ]
    }
    SizeBox_210 {
     id=#2
     possessed_class=SizeBox
     tracks [{
       type=MovieScene2DTransformTrack
       sections [{
         type=MovieScene2DTransformSection
         channels [
          Translation.X {type=MovieSceneScriptingFloatChannel}
          Translation.Y {type=MovieSceneScriptingFloatChannel}
          Angle {type=MovieSceneScriptingFloatChannel}
          Scale.X {type=MovieSceneScriptingFloatChannel}
          Scale.Y {type=MovieSceneScriptingFloatChannel, keys [[0,0],[4,1]]}
          Shear.X {type=MovieSceneScriptingFloatChannel}
          Shear.Y {type=MovieSceneScriptingFloatChannel}
         ]
        }]
      }]
    }
    ScrollBox_Content {
     id=#3
     possessed_class=ScrollBox
     tracks [
      {
       type=MovieScene2DTransformTrack
       sections [{
         type=MovieScene2DTransformSection
         channels [
          Translation.X {type=MovieSceneScriptingFloatChannel, keys [[5,128],[10,0]]}
          Translation.Y {type=MovieSceneScriptingFloatChannel}
          Angle {type=MovieSceneScriptingFloatChannel}
          Scale.X {type=MovieSceneScriptingFloatChannel}
          Scale.Y {type=MovieSceneScriptingFloatChannel}
          Shear.X {type=MovieSceneScriptingFloatChannel}
          Shear.Y {type=MovieSceneScriptingFloatChannel}
         ]
        }]
      }
      {
       type=MovieSceneFloatTrack
       display_name=Render Opacity
       sections [{type=MovieSceneFloatSection, channels [None {type=MovieSceneScriptingFloatChannel, keys [[5,0],[8,1]]}]}]
      }
     ]
    }
    Button_Accept {
     id=#4
     possessed_class=Spatial_Specialbutton_C
     tracks [{
       type=MovieScene2DTransformTrack
       sections [{
         type=MovieScene2DTransformSection
         channels [
          Translation.X {type=MovieSceneScriptingFloatChannel}
          Translation.Y {type=MovieSceneScriptingFloatChannel}
          Angle {type=MovieSceneScriptingFloatChannel}
          Scale.X {type=MovieSceneScriptingFloatChannel, keys [[3,0],[5,1],[8,1]]}
          Scale.Y {type=MovieSceneScriptingFloatChannel, keys [[3,0],[5,1],[8,1]]}
          Shear.X {type=MovieSceneScriptingFloatChannel}
          Shear.Y {type=MovieSceneScriptingFloatChannel}
         ]
        }]
      }]
    }
    Image_69 {
     id=#5
     possessed_class=Image
     tracks [{
       type=MovieSceneWidgetMaterialTrack
       sections [{
         type=MovieSceneParameterSection
         end_frame=20
         start_frame=0
         channels [Shape_Offset_Top {type=MovieSceneScriptingFloatChannel, keys [[0,-1],[8,-0.25]]}]
        }]
      }]
    }
   ]}
 }]
graphs [
 EventGraph {nodes [
   n1 = "On Button Clicked (Button_Accept)"!() -> (OutputDelegate, then, Button:Common Button Legacy Object Reference) {is=title K2Node_ComponentBoundEvent}
   Call Close About!(execute:Exec=[@n1.then, @n2.then], self:BPS12 About Modal Object Reference) -> (then:Exec) {is=title K2Node_CallDelegate}
   > FadeOut!(
    execute
    self:Audio Component Object Reference=Get AboutModalAudio() -> (AboutModalAudio:Audio Component Object Reference)
    FadeOutDuration:float=0.5
    FadeVolumeLevel:float=0
    FadeCurve:EAudioFaderCurve Enum=Linear
   ) -> (then)
   n2 = "On Button Clicked (Button_CloseTouch)"!() -> (OutputDelegate, then, Button:Common Button Legacy Object Reference) {is=title K2Node_ComponentBoundEvent}
   n3 = Play open anim() {is=title K2Node_CustomEvent}
   PlayAnimationForward(execute=[@n3.then, @n5.then], InAnimation=Get ShowUp() -> (ShowUp:Widget Animation Object Reference))
   Play Intro Animation Sound() {is=title K2Node_CustomEvent}
   > n4 = SpawnSound2D!(
    execute
    Sound:Sound Base Object Reference=/BRCosmetics/Sounds/Other/Fort_UI_Sounds/FrontEnd/UI_S14_AboutModal_PopUp_Cue
    VolumeMultiplier:float=1
    PitchMultiplier:float=1
    StartTime:float=0
    ConcurrencySettings:Sound Concurrency Object Reference
    bPersistAcrossLevelTransition:bool=false
    bAutoDestroy:bool=true
   ) -> (then, ReturnValue:Audio Component Object Reference)
   > Set AboutModalAudio(AboutModalAudio:Audio Component Object Reference=@n4) -> (Output_Get:Audio Component Object Reference)
   Event PreConstruct()
   > SetupContent!(execute, self, InTitle:Text=Get Title() -> (Title:Text), InEntries:Array of Texts=Get Entries() -> (Entries:Array of Texts)) -> (then)
   > SetupStyling()
   > SetVisibility(
    self=Get ModalHeader() -> (ModalHeader:Horizontal Box Object Reference)
    InVisibility=Select(Option 0=Collapsed, Option 1=HitTestInvisible, Index=Get show Header() -> (show Header:bool))
   )
   n5 = Event On Activated()
   Event Construct()
   > BindToAnimationStarted!(execute, self:User Widget Object Reference, Animation:Widget Animation Object Reference=Get ShowUp(), Delegate:Delegate=Create Event()) -> (then)
   n6 = Event OnSetContent!() -> (OutputDelegate, then, InTitle:"Text (by ref)", InEntries:Array of Texts) {is=title K2Node_Event}
   > SetupContent(InTitle=@n6.InTitle, InEntries=@n6.InEntries)
  ]}
 SetupContent {nodes [
   SetupContent!() -> (then, InTitle:Text, InEntries:Array of Texts) {is=title K2Node_FunctionEntry}
   > ClearChildren(self=Get VerticalBox_DescriptionReceiver() -> (VerticalBox_DescriptionReceiver:Vertical Box Object Reference))
   > "SetText (Text)"(self=Get Title_white() -> (Title_white:Common Text Object Reference), InText=Get InTitle() -> (InTitle:Text))
   > SetupEntries!(execute, self, InEntries:Array of Texts=Get InEntries() -> (InEntries:Array of Texts)) -> (then)
   > "SetText (Text)"(self=Get HeaderMainText() -> (HeaderMainText:Common Text Object Reference), InText=Get Header() -> (Header:Text))
   > "SetText (Text)"(self=Get SubHeaderText() -> (SubHeaderText:Common Text Object Reference), InText=Get SubHeader() -> (SubHeader:Text))
   > SetContentColorAndOpacity!(
    execute
    self:Border Object Reference=Get SubtextColor() -> (SubtextColor:Border Object Reference)
    InContentColorAndOpacity:Linear Color Structure=Get Subheader_Color() -> (Subheader_Color:Linear Color Structure)
   ) -> (then)
   > SetContentColorAndOpacity(
    self=Get HeaderColor() -> (HeaderColor:Border Object Reference)
    InContentColorAndOpacity=Get Header_Color() -> (Header_Color:Linear Color Structure)
   )
   > SetColorAndOpacity(
    self:Image Object Reference=Get DividerLine() -> (DividerLine:Image Object Reference)
    InColorAndOpacity:Linear Color Structure=Get Divider_Color() -> (Divider_Color:Linear Color Structure)
   )
   > SetBrushFromMaterial(
    self=Get HeaderIcon() -> (HeaderIcon:Image Object Reference)
    Material=Get Header_Icon() -> (Header_Icon:Material Interface Object Reference)
   )
   > SetVisibility(
    self=Get HeaderIcon()
    InVisibility=Select(Option 0=Collapsed, Option 1=SelfHitTestInvisible, Index=Get ShowHeaderIcon() -> (ShowHeaderIcon:bool))
   )
   > SetVisibility(
    self=Get DividerLine()
    InVisibility=Select(Option 0=Collapsed, Option 1=SelfHitTestInvisible, Index=Get ShowDividerLine() -> (ShowDividerLine:bool))
   )
   > SetVisibility(self=Get ModalHeader(), InVisibility=Select(Option 0=Collapsed, Option 1=SelfHitTestInvisible, Index=Get ShowHeader() -> (ShowHeader:bool)))
  ]}
 SetupEntries {nodes [
   n1 = SetupEntries!() -> (then, InEntries:Array of Texts) {is=title K2Node_FunctionEntry}
   > n2 = For Each Loop(Array:Array of Texts=@n1) -> (Array Element:Text)
   > n3 = Create Widget!(
    execute:Exec
    Class:User Widget Class Reference=/FortniteCoreUI/Frontend/Spatial/BPS12_AboutModal_Entry.BPS12_AboutModal_Entry_C
    OwningPlayer:Player Controller Object Reference
    text:Text=@n2
    GoldnBlack:bool=Get GoldNblack() -> (GoldNblack:bool)
   ) -> (then:Exec, ReturnValue:BPS12 About Modal Entry Object Reference)
   > AddChild(self=Get VerticalBox_DescriptionReceiver(), Content=@n3)
  ]}
 SetupStyling {nodes [
   SetupStyling() {is=title K2Node_FunctionEntry}
   n1 = Get ScrollBox_Content() -> (ScrollBox_Content:Scroll Box Object Reference)
   n2 = Break Scroll Bar Style!(ScrollBarStyle:"Scroll Bar Style Structure (by ref)"=Get WidgetBarStyle!(self:Scroll Box Object Reference=@n1) -> (WidgetBarStyle:Scroll Bar Style Structure)) -> (
    HorizontalBackgroundImage:Slate Brush Structure
    VerticalBackgroundImage:Slate Brush Structure
    VerticalTopSlotImage:Slate Brush Structure
    HorizontalTopSlotImage:Slate Brush Structure
    VerticalBottomSlotImage:Slate Brush Structure
    HorizontalBottomSlotImage:Slate Brush Structure
    NormalThumbImage:Slate Brush Structure
    HoveredThumbImage:Slate Brush Structure
    DraggedThumbImage:Slate Brush Structure
    Thickness:float=~
   ) {is=title K2Node_BreakStruct}
   n3 = Break Slate Brush!(SlateBrush:"Slate Brush Structure (by ref)"=@n2.NormalThumbImage) -> (
    TintColor:Slate Color Structure
    DrawAs:ESlateBrushDrawType Enum
    Tiling:ESlateBrushTileType Enum
    Mirroring:ESlateBrushMirrorType Enum
    ImageSize:Vector 2D Structure
    Margin:Margin Structure
    ResourceObject:Object Reference
    OutlineSettings:Slate Brush Outline Settings Structure
   ) {is=title K2Node_BreakStruct}
   n4 = Select(
    Option 0:Slate Color Structure=Make Slate Color(SpecifiedColor=(R=0.0331,G=0.855,B=1,A=1))
    Option 1:Slate Color Structure=Make Slate Color(SpecifiedColor=(R=0.7157,G=0.4793,B=0.0742,A=1))
    Index=Get GoldNblack()
   ) -> (ReturnValue:Slate Color Structure)
   > Set WidgetBarStyle!(
    execute
    WidgetBarStyle:Scroll Bar Style Structure=Make Scroll Bar Style!(
     HorizontalBackgroundImage:Slate Brush Structure=@n2
     VerticalBackgroundImage:Slate Brush Structure=@n2.VerticalBackgroundImage
     VerticalTopSlotImage:Slate Brush Structure=@n2.VerticalTopSlotImage
     HorizontalTopSlotImage:Slate Brush Structure=@n2.HorizontalTopSlotImage
     VerticalBottomSlotImage:Slate Brush Structure=@n2.VerticalBottomSlotImage
     HorizontalBottomSlotImage:Slate Brush Structure=@n2.HorizontalBottomSlotImage
     NormalThumbImage:Slate Brush Structure=Make Slate Brush!(
      TintColor:Slate Color Structure=@n4
      DrawAs:ESlateBrushDrawType Enum=@n3.DrawAs
      Tiling:ESlateBrushTileType Enum=@n3.Tiling
      Mirroring:ESlateBrushMirrorType Enum=@n3.Mirroring
      ImageSize:Vector 2D Structure=@n3.ImageSize
      Margin:Margin Structure=@n3.Margin
      ResourceObject:Object Reference=@n3.ResourceObject
      OutlineSettings:Slate Brush Outline Settings Structure=(CornerRadii=(X=0,Y=0,Z=0,W=0),Color=(SpecifiedColor=(R=0,G=0,B=0,A=0),ColorUseRule=UseColor_Specified),Width=0,RoundingType=HalfHeightRadius,bUseBrushTransparency=False)
     ) -> (SlateBrush:Slate Brush Structure) {is=title K2Node_MakeStruct}
     HoveredThumbImage:Slate Brush Structure=@n2.HoveredThumbImage
     DraggedThumbImage:Slate Brush Structure=@n2.DraggedThumbImage
     Thickness:float=16
    ) -> (ScrollBarStyle:Scroll Bar Style Structure) {is=title K2Node_MakeStruct}
    self:Scroll Box Object Reference=@n1
   ) -> (then, Output_Get:Scroll Bar Style Structure)
   > SetHeightOverride(self=Get SizeBox_210() -> (SizeBox_210:Size Box Object Reference), InHeightOverride=Get Height Override() -> (Height Override:double))
  ]}
 closeAbout {nodes [closeAbout() {is=title K2Node_FunctionEntry}]}
]
variables [
 AboutModalAudio {type=AudioComponent, default=null}
 closeAbout {}
 Divider_Color {type=LinearColor, default=<rgba 1 1 1 1>}
 GoldNblack {type=boolean}
 Header {type=string, default=Header}
 Header_Color {type=LinearColor, default=<rgba 1 1 1 1>}
 Header_Icon {type=MaterialInterface, default=/BRCosmetics/Athena/UI/Frontend/Spatial/Art/M_UI_UmbrellaIcon_Base}
 Height Override {type=number, default=512}
 show Header {type=boolean, default=true}
 ShowDividerLine {type=boolean, default=true}
 ShowHeader {type=boolean, default=true}
 ShowHeaderIcon {type=boolean, default=true}
 SubHeader {type=string, default=Sub header text}
 Subheader_Color {type=LinearColor, default=<rgba 1 1 1 1>}
]
widgets {
 info {inheritedWidgetCount=2, parentClass=/Script/FortniteUI.AthenaAboutModal, rootWidgetClass=&Overlay, widgetCount=33}
 widgets [
  Overlay_0 {widgetClassPath=&Overlay, widget=Overlay_0}
  BackgroundBlur_20 {
   widgetClassPath=&BackgroundBlur
   parent=Overlay_0
   widget=BackgroundBlur_20
   properties {lowQualityFallbackBrush {
     bIsDynamicallyLoaded=false
     drawAs=Image
     imageType=NoImage
     mirroring=NoMirror
     resourceName=None
     resourceObject=null
     tiling=NoTile
     imageSize=<0 0>
     margin=<ltrb 0 0 0 0>
     outlineSettings {
      bUseBrushTransparency=false
      roundingType=HalfHeightRadius
      width=0
      color {colorUseRule=UseColor_Specified, specifiedColor=<rgba 0 0 0 0>}
      cornerRadii=<0 0 0 0>
     }
     tintColor {colorUseRule=UseColor_Specified, specifiedColor=<rgba 0 0 0 0.75>}
     uVRegion {bIsValid=false, max=<0 0>, min=<0 0>}
    }}
   slot {class=OverlaySlot, properties {horizontalAlignment=HAlign_Fill, verticalAlignment=VAlign_Fill, padding=<ltrb -2 -2 -2 -2>}}
  }
  Image_42 {
   widgetClassPath=&Image
   bIsVariable=true
   parent=BackgroundBlur_20
   widget=Image_42
   properties {
    visibility=Collapsed
    brush {
     bIsDynamicallyLoaded=false
     drawAs=Image
     imageType=NoImage
     mirroring=NoMirror
     resourceName=None
     resourceObject=/Game/Athena/UI/Frontend/Art/M_UI_RadialGradientMaster_Dark
     tiling=NoTile
     imageSize=<32 32>
     margin=<ltrb 0 0 0 0>
     outlineSettings {
      bUseBrushTransparency=false
      roundingType=HalfHeightRadius
      width=0
      color {colorUseRule=UseColor_Specified, specifiedColor=<rgba 0 0 0 0>}
      cornerRadii=<0 0 0 0>
     }
     tintColor {colorUseRule=UseColor_Specified, specifiedColor=<rgba 0.6 0.6 0.6 1>}
     uVRegion {bIsValid=false, max=<0 0>, min=<0 0>}
    }
   }
   slot {class=BackgroundBlurSlot, properties {padding=<ltrb 0 0 0 0>}}
  }
  SizeBox_210 {
   widgetClassPath=&SizeBox
   bIsVariable=true
   parent=Overlay_0
   widget=SizeBox_210
   properties {bOverride_HeightOverride=true, heightOverride=512}
   slot {class=OverlaySlot, properties {horizontalAlignment=HAlign_Fill, verticalAlignment=VAlign_Center, padding=<ltrb -2 0 -2 0>}}
  }
  Overlay_57 {widgetClassPath=&Overlay, parent=SizeBox_210, widget=Overlay_57, slot {class=SizeBoxSlot}}
  Image_110 {
   widgetClassPath=&Image
   bIsVariable=true
   parent=Overlay_57
   widget=Image_110
   properties {brush {
     bIsDynamicallyLoaded=false
     drawAs=Image
     imageType=NoImage
     mirroring=NoMirror
     resourceName=None
     resourceObject=null
     tiling=NoTile
     imageSize=<32 32>
     margin=<ltrb 0 0 0 0>
     outlineSettings {
      bUseBrushTransparency=false
      roundingType=HalfHeightRadius
      width=0
      color {colorUseRule=UseColor_Specified, specifiedColor=<rgba 0 0 0 0>}
      cornerRadii=<0 0 0 0>
     }
     tintColor {colorUseRule=UseColor_Specified, specifiedColor=<rgba 0.0262 0 0.0097 0.5>}
     uVRegion {bIsValid=false, max=<0 0>, min=<0 0>}
    }}
   slot {class=OverlaySlot, properties {horizontalAlignment=HAlign_Fill, verticalAlignment=VAlign_Fill}}
  }
  Image_69 {
   widgetClassPath=&Image
   bIsVariable=true
   parent=Overlay_57
   widget=Image_69
   properties {brush {
     bIsDynamicallyLoaded=false
     drawAs=Image
     imageType=NoImage
     mirroring=NoMirror
     resourceName=None
     resourceObject=/FortUILibrary/Foundation/Art_ToolKit/AngledBoxes/M_UI_MoreInfo_BottomDividerLIne
     tiling=NoTile
     imageSize=<32 26>
     margin=<ltrb 0 0 0 0>
     outlineSettings {
      bUseBrushTransparency=false
      roundingType=HalfHeightRadius
      width=0
      color {colorUseRule=UseColor_Specified, specifiedColor=<rgba 0 0 0 0>}
      cornerRadii=<0 0 0 0>
     }
     tintColor {colorUseRule=UseColor_Specified, specifiedColor=<rgba 1 1 1 1>}
     uVRegion {bIsValid=false, max=<0 0>, min=<0 0>}
    }}
   slot {class=OverlaySlot, properties {horizontalAlignment=HAlign_Fill, verticalAlignment=VAlign_Bottom}}
  }
  SafeZone_64 {
   widgetClassPath=&SafeZone
   bIsVariable=true
   parent=Overlay_57
   widget=SafeZone_64
   slot {class=OverlaySlot, properties {horizontalAlignment=HAlign_Fill, verticalAlignment=VAlign_Fill}}
  }
  HorizontalBox_0 {
   widgetClassPath=&HorizontalBox
   parent=SafeZone_64
   widget=HorizontalBox_0
   slot {class=SafeZoneSlot, properties {padding=<ltrb 88 0 150 0>}}
  }
  VerticalBox_93 {
   widgetClassPath=&VerticalBox
   parent=HorizontalBox_0
   widget=VerticalBox_93
   slot {class=HorizontalBoxSlot, properties {padding=<ltrb 0 88 0 32>, size {sizeRule=Fill, value=1}}}
  }
  SizeBox_0 {
   widgetClassPath=&SizeBox
   parent=VerticalBox_93
   widget=SizeBox_0
   properties {bOverride_WidthOverride=true, maxDesiredWidth=512, widthOverride=768}
   slot {class=VerticalBoxSlot, properties {horizontalAlignment=HAlign_Left, padding=<ltrb 0 0 0 24>}}
  }
  ScaleBox_0 {
   widgetClassPath=&ScaleBox
   parent=SizeBox_0
   widget=ScaleBox_0
   properties {stretch=ScaleToFitX, stretchDirection=DownOnly}
   slot {class=SizeBoxSlot}
  }
  Title_white {
   widgetClassPath=&CommonTextBlock
   bIsVariable=true
   parent=ScaleBox_0
   widget=Title_white
   properties {
    text="Make Maya Uniquely Yours! "
    textTransformPolicy=ToUpper
    colorAndOpacity {colorUseRule=UseColor_Specified, specifiedColor=<rgba 1 0 0.0395 1>}
    font {
     bForceMonospaced=false
     bMaterialIsStencil=false
     fontMaterial=null
     fontObject=&Burbank
     letterSpacing=0
     monospacedWidth=1
     size=48
     skewAmount=0
     typefaceFontName=BurbankRegular-Black
     outlineSettings {
      bApplyOutlineToDropShadows=false
      bMiteredCorners=false
      bSeparateFillAlpha=false
      outlineMaterial=null
      outlineSize=0
      outlineColor=<rgba 1 1 1 0>
     }
    }
    margin=<ltrb 16 0 16 0>
    renderTransform {angle=0, scale=<1 1>, shear=<-12 0>, translation=<0 0>}
    shadowOffset=<0 0>
   }
   slot {class=ScaleBoxSlot, properties {horizontalAlignment=HAlign_Left}}
  }
  ScrollBox_Content {
   widgetClassPath=&ScrollBox
   bInherited=true
   parent=VerticalBox_93
   widget=ScrollBox_Content
   properties {
    scrollbarPadding=<ltrb 16 2 2 2>
    scrollbarThickness=<12 12>
    widgetBarStyle {
     thickness=16
     draggedThumbImage {
      bIsDynamicallyLoaded=false
      drawAs=Box
      imageType=FullColor
      mirroring=NoMirror
      resourceName=../../../Engine/Content/Slate/Common/Scrollbar_Thumb.png
      resourceObject=/BRCosmetics/Athena/UI/Frontend/Spatial/Art/M_Ui_Spacial_Seaparator_Background_progress_Inst
      tiling=NoTile
      imageSize=<8 8>
      margin=<ltrb 0.45 0.45 0.45 0.45>
      outlineSettings {
       bUseBrushTransparency=false
       roundingType=HalfHeightRadius
       width=0
       color {colorUseRule=UseColor_Specified, specifiedColor=<rgba 0 0 0 0>}
       cornerRadii=<0 0 0 0>
      }
      tintColor {colorUseRule=UseColor_Specified, specifiedColor=<rgba 1 1 1 1>}
      uVRegion {bIsValid=false, max=<0 0>, min=<0 0>}
     }
     horizontalBackgroundImage {
      bIsDynamicallyLoaded=false
      drawAs=NoDrawType
      imageType=NoImage
      mirroring=NoMirror
      resourceName=None
      resourceObject=null
      tiling=NoTile
      imageSize=<0 0>
      margin=<ltrb 0 0 0 0>
      outlineSettings {
       bUseBrushTransparency=false
       roundingType=HalfHeightRadius
       width=0
       color {colorUseRule=UseColor_Specified, specifiedColor=<rgba 0 0 0 0>}
       cornerRadii=<0 0 0 0>
      }
      tintColor {colorUseRule=UseColor_Specified, specifiedColor=<rgba 1 1 1 1>}
      uVRegion {bIsValid=false, max=<0 0>, min=<0 0>}
     }
     horizontalBottomSlotImage {
      bIsDynamicallyLoaded=false
      drawAs=NoDrawType
      imageType=FullColor
      mirroring=NoMirror
      resourceName=../../../Engine/Content/Slate/Common/Scrollbar_Background_Horizontal.png
      resourceObject=null
      tiling=NoTile
      imageSize=<8 8>
      margin=<ltrb 0 0 0 0>
      outlineSettings {
       bUseBrushTransparency=false
       roundingType=HalfHeightRadius
       width=0
       color {colorUseRule=UseColor_Specified, specifiedColor=<rgba 0 0 0 0>}
       cornerRadii=<0 0 0 0>
      }
      tintColor {colorUseRule=UseColor_Specified, specifiedColor=<rgba 1 1 1 1>}
      uVRegion {bIsValid=false, max=<0 0>, min=<0 0>}
     }
     horizontalTopSlotImage {
      bIsDynamicallyLoaded=false
      drawAs=NoDrawType
      imageType=FullColor
      mirroring=NoMirror
      resourceName=../../../Engine/Content/Slate/Common/Scrollbar_Background_Horizontal.png
      resourceObject=null
      tiling=NoTile
      imageSize=<8 8>
      margin=<ltrb 0 0 0 0>
      outlineSettings {
       bUseBrushTransparency=false
       roundingType=HalfHeightRadius
       width=0
       color {colorUseRule=UseColor_Specified, specifiedColor=<rgba 0 0 0 0>}
       cornerRadii=<0 0 0 0>
      }
      tintColor {colorUseRule=UseColor_Specified, specifiedColor=<rgba 1 1 1 1>}
      uVRegion {bIsValid=false, max=<0 0>, min=<0 0>}
     }
     hoveredThumbImage {
      bIsDynamicallyLoaded=false
      drawAs=Box
      imageType=FullColor
      mirroring=NoMirror
      resourceName=../../../Engine/Content/Slate/Common/Scrollbar_Thumb.png
      resourceObject=&M_Ui_Spacial_Seaparator_Background_progress_Inst
      tiling=NoTile
      imageSize=<8 8>
      margin=<ltrb 0.45 0.45 0.45 0.45>
      outlineSettings {
       bUseBrushTransparency=false
       roundingType=HalfHeightRadius
       width=0
       color {colorUseRule=UseColor_Specified, specifiedColor=<rgba 0 0 0 0>}
       cornerRadii=<0 0 0 0>
      }
      tintColor {colorUseRule=UseColor_Specified, specifiedColor=<rgba 1 1 1 1>}
      uVRegion {bIsValid=false, max=<0 0>, min=<0 0>}
     }
     normalThumbImage {
      bIsDynamicallyLoaded=false
      drawAs=Box
      imageType=FullColor
      mirroring=NoMirror
      resourceName=../../../Engine/Content/Slate/Common/Scrollbar_Thumb.png
      resourceObject=&M_Ui_Spacial_Seaparator_Background_progress_Inst
      tiling=NoTile
      imageSize=<8 8>
      margin=<ltrb 0.45 0.45 0.45 0.45>
      outlineSettings {
       bUseBrushTransparency=false
       roundingType=HalfHeightRadius
       width=0
       color {colorUseRule=UseColor_Specified, specifiedColor=<rgba 0 0 0 0>}
       cornerRadii=<0 0 0 0>
      }
      tintColor {colorUseRule=UseColor_Specified, specifiedColor=<rgba 0.0331 0.855 1 1>}
      uVRegion {bIsValid=false, max=<0 0>, min=<0 0>}
     }
     verticalBackgroundImage {
      bIsDynamicallyLoaded=false
      drawAs=Box
      imageType=NoImage
      mirroring=NoMirror
      resourceName=None
      resourceObject=&M_Ui_Spacial_Seaparator_Background_progress_Inst
      tiling=NoTile
      imageSize=<8 8>
      margin=<ltrb 0.45 0.45 0.45 0.45>
      outlineSettings {
       bUseBrushTransparency=false
       roundingType=HalfHeightRadius
       width=0
       color {colorUseRule=UseColor_Specified, specifiedColor=<rgba 0 0 0 0>}
       cornerRadii=<0 0 0 0>
      }
      tintColor {colorUseRule=UseColor_Specified, specifiedColor=<rgba 0 0 0 0.25>}
      uVRegion {bIsValid=false, max=<0 0>, min=<0 0>}
     }
     verticalBottomSlotImage {
      bIsDynamicallyLoaded=false
      drawAs=NoDrawType
      imageType=FullColor
      mirroring=NoMirror
      resourceName=../../../Engine/Content/Slate/Common/Scrollbar_Background_Vertical.png
      resourceObject=null
      tiling=NoTile
      imageSize=<8 8>
      margin=<ltrb 0 0 0 0>
      outlineSettings {
       bUseBrushTransparency=false
       roundingType=HalfHeightRadius
       width=0
       color {colorUseRule=UseColor_Specified, specifiedColor=<rgba 0 0 0 0>}
       cornerRadii=<0 0 0 0>
      }
      tintColor {colorUseRule=UseColor_Specified, specifiedColor=<rgba 1 1 1 1>}
      uVRegion {bIsValid=false, max=<0 0>, min=<0 0>}
     }
     verticalTopSlotImage {
      bIsDynamicallyLoaded=false
      drawAs=NoDrawType
      imageType=FullColor
      mirroring=NoMirror
      resourceName=../../../Engine/Content/Slate/Common/Scrollbar_Background_Vertical.png
      resourceObject=null
      tiling=NoTile
      imageSize=<8 8>
      margin=<ltrb 0 0 0 0>
      outlineSettings {
       bUseBrushTransparency=false
       roundingType=HalfHeightRadius
       width=0
       color {colorUseRule=UseColor_Specified, specifiedColor=<rgba 0 0 0 0>}
       cornerRadii=<0 0 0 0>
      }
      tintColor {colorUseRule=UseColor_Specified, specifiedColor=<rgba 1 1 1 1>}
      uVRegion {bIsValid=false, max=<0 0>, min=<0 0>}
     }
    }
    widgetStyle {
     barThickness=9
     bottomShadowBrush {
      bIsDynamicallyLoaded=false
      drawAs=Box
      imageType=FullColor
      mirroring=NoMirror
      resourceName=../../../Engine/Content/Slate/Common/ScrollBoxShadowBottom.png
      resourceObject=null
      tiling=NoTile
      imageSize=<16 8>
      margin=<ltrb 0.5 0 0.5 1>
      outlineSettings {
       bUseBrushTransparency=false
       roundingType=HalfHeightRadius
       width=0
       color {colorUseRule=UseColor_Specified, specifiedColor=<rgba 0 0 0 0>}
       cornerRadii=<0 0 0 0>
      }
      tintColor {colorUseRule=UseColor_Specified, specifiedColor=<rgba 1 1 1 1>}
      uVRegion {bIsValid=false, max=<0 0>, min=<0 0>}
     }
     horizontalScrolledContentPadding=<ltrb 0 0 1 0>
     leftShadowBrush {
      bIsDynamicallyLoaded=false
      drawAs=NoDrawType
      imageType=FullColor
      mirroring=NoMirror
      resourceName=../../../Engine/Content/Slate/Common/ScrollBoxShadowLeft.png
      resourceObject=null
      tiling=NoTile
      imageSize=<8 16>
      margin=<ltrb 1 0.5 0 0.5>
      outlineSettings {
       bUseBrushTransparency=false
       roundingType=HalfHeightRadius
       width=0
       color {colorUseRule=UseColor_Specified, specifiedColor=<rgba 0 0 0 0>}
       cornerRadii=<0 0 0 0>
      }
      tintColor {colorUseRule=UseColor_Specified, specifiedColor=<rgba 1 1 1 1>}
      uVRegion {bIsValid=false, max=<0 0>, min=<0 0>}
     }
     rightShadowBrush {
      bIsDynamicallyLoaded=false
      drawAs=NoDrawType
      imageType=FullColor
      mirroring=NoMirror
      resourceName=../../../Engine/Content/Slate/Common/ScrollBoxShadowRight.png
      resourceObject=null
      tiling=NoTile
      imageSize=<8 16>
      margin=<ltrb 0 0.5 1 0.5>
      outlineSettings {
       bUseBrushTransparency=false
       roundingType=HalfHeightRadius
       width=0
       color {colorUseRule=UseColor_Specified, specifiedColor=<rgba 0 0 0 0>}
       cornerRadii=<0 0 0 0>
      }
      tintColor {colorUseRule=UseColor_Specified, specifiedColor=<rgba 1 1 1 1>}
      uVRegion {bIsValid=false, max=<0 0>, min=<0 0>}
     }
     topShadowBrush {
      bIsDynamicallyLoaded=false
      drawAs=Box
      imageType=FullColor
      mirroring=NoMirror
      resourceName=../../../Engine/Content/Slate/Common/ScrollBoxShadowTop.png
      resourceObject=null
      tiling=NoTile
      imageSize=<16 8>
      margin=<ltrb 0.5 1 0.5 0>
      outlineSettings {
       bUseBrushTransparency=false
       roundingType=HalfHeightRadius
       width=0
       color {colorUseRule=UseColor_Specified, specifiedColor=<rgba 0 0 0 0>}
       cornerRadii=<0 0 0 0>
      }
      tintColor {colorUseRule=UseColor_Specified, specifiedColor=<rgba 1 1 1 1>}
      uVRegion {bIsValid=false, max=<0 0>, min=<0 0>}
     }
     verticalScrolledContentPadding=<ltrb 0 0 0 1>
    }
   }
   slot {class=VerticalBoxSlot, properties {size {sizeRule=Fill, value=1}}}
  }
  VerticalBox_DescriptionReceiver {
   widgetClassPath=&VerticalBox
   bIsVariable=true
   parent=ScrollBox_Content
   widget=VerticalBox_DescriptionReceiver
   slot {class=ScrollBoxSlot}
  }
  BPS12_AboutModal_Entry {
   widgetClassPath=/FortniteCoreUI/Frontend/Spatial/BPS12_AboutModal_Entry
   bIsVariable=true
   parent=VerticalBox_DescriptionReceiver
   widget=BPS12_AboutModal_Entry
   properties {text="Purchase the <text color=\"FFFFFFFF\" fontface=\"black\">Season 2 Battle Pass</> and visit Maya's Upgrade Vault"}
   slot {class=VerticalBoxSlot}
  }
  BPS12_AboutModal_Entry_1 {
   widgetClassPath=&BPS12_AboutModal_Entry
   bIsVariable=true
   parent=VerticalBox_DescriptionReceiver
   widget=BPS12_AboutModal_Entry_1
   properties {text="Complete <text color=\"FFFFFFFF\" fontface=\"black\">Maya's Challenge</> each week to be able to upgrade different parts of her outfit"}
   slot {class=VerticalBoxSlot}
  }
  BPS12_AboutModal_Entry_107 {
   widgetClassPath=&BPS12_AboutModal_Entry
   bIsVariable=true
   parent=VerticalBox_DescriptionReceiver
   widget=BPS12_AboutModal_Entry_107
   properties {text="Preview your upgrade choices and then <text color=\"FFFFFFFF\" fontface=\"black\">permanently lock them in</> as you finish challenges to craft your own unique version of Maya"}
   slot {class=VerticalBoxSlot}
  }
  Overlay_59 {
   widgetClassPath=&Overlay
   parent=HorizontalBox_0
   widget=Overlay_59
   slot {class=HorizontalBoxSlot, properties {verticalAlignment=VAlign_Bottom, padding=<ltrb 150 0 0 0>}}
  }
  Image_Item {
   widgetClassPath=&FortLazyImage
   bInherited=true
   bIsVariable=true
   parent=Overlay_59
   widget=Image_Item
   properties {
    materialTextureParamName=ItemIcon
    visibility=HitTestInvisible
    brush {
     bIsDynamicallyLoaded=false
     drawAs=Image
     imageType=NoImage
     mirroring=NoMirror
     resourceName=None
     resourceObject=null
     tiling=NoTile
     imageSize=<448 448>
     margin=<ltrb 0 0 0 0>
     outlineSettings {
      bUseBrushTransparency=false
      roundingType=HalfHeightRadius
      width=0
      color {colorUseRule=UseColor_Specified, specifiedColor=<rgba 0 0 0 0>}
      cornerRadii=<0 0 0 0>
     }
     tintColor {colorUseRule=UseColor_Specified, specifiedColor=<rgba 1 1 1 1>}
     uVRegion {bIsValid=false, max=<0 0>, min=<0 0>}
    }
    renderTransform {angle=0, scale=<1.5 1.5>, shear=<0 0>, translation=<0 0>}
    renderTransformPivot=<0.5 1>
   }
   slot {class=OverlaySlot, properties {padding=<ltrb 0 0 64 0>}}
  }
  SizeBox_212 {
   widgetClassPath=&SizeBox
   parent=Overlay_57
   widget=SizeBox_212
   properties {
    bOverride_HeightOverride=true
    bOverride_MinDesiredWidth=true
    heightOverride=128
    minDesiredWidth=416
    renderTransform {angle=0, scale=<1 1>, shear=<0 0>, translation=<0 64>}
   }
   slot {class=OverlaySlot, properties {horizontalAlignment=HAlign_Center, verticalAlignment=VAlign_Bottom, padding=<ltrb 20 20 10 0>}}
  }
  Button_Accept {
   widgetClassPath=/FortniteCoreUI/Frontend/Spatial/Customization/Spatial_Specialbutton
   bIsVariable=true
   parent=SizeBox_212
   widget=Button_Accept
   properties {
    auto wrap text=false
    button text="GOT IT!"
    fontSize=40
    fontSizeOverride=true
    justifiction=Center
    material Brush=/Game/Athena/UI/Frontend/Spatial/Art/M_Ui_Skewed_SpatialButton_MoreInfo
    onButtonBaseClicked=(&self:WidgetTree.Button_Accept.HandleOnButtonClicked)
    onButtonBaseDoubleClicked=(&self:WidgetTree.Button_Accept.HandleOnButtonDoubleClicked)
    onButtonBaseHovered=(&self:WidgetTree.Button_Accept.HandleOnButtonHovered)
    onButtonBaseUnhovered=(&self:WidgetTree.Button_Accept.HandleOnButtonUnhovered)
    onSelectedChangedBase=(&self:WidgetTree.Button_Accept.HandleOnSelectedChanged)
    right Image=null
    shearText=true
    simpleHover=true
    hoveredSlateSoundOverride {resourceObject=/Game/Sounds/Fort_UI_Sounds/FrontEnd/DefaultSounds/PressedButtons/MSS_UI_Button_Hover_Default}
    pressedSlateSoundOverride {resourceObject=/BRCosmetics/Sounds/Other/Fort_UI_Sounds/FrontEnd/New_FrontEnd/Fortnite_UI_RogerThat_Pressed_Cue}
    text Color {colorUseRule=UseColor_Specified, specifiedColor=<rgba 0.75 1 1 1>}
    triggeringInputAction {dataTable=&InputActionDataTable, rowName=GenericFaceButtonBottom}
   }
   slot {class=SizeBoxSlot}
  }
  ModalHeader {
   widgetClassPath=&HorizontalBox
   bIsVariable=true
   parent=Overlay_57
   widget=ModalHeader
   properties {renderTransform {angle=0, scale=<1 1>, shear=<0 0>, translation=<0 -144>}}
   slot {class=OverlaySlot, properties {padding=<ltrb 80 0 0 0>}}
  }
  HeaderIcon {
   widgetClassPath=&Image
   bIsVariable=true
   parent=ModalHeader
   widget=HeaderIcon
   properties {brush {
     bIsDynamicallyLoaded=false
     drawAs=Image
     imageType=NoImage
     mirroring=NoMirror
     resourceName=None
     resourceObject=&M_UI_UmbrellaIcon_Base
     tiling=NoTile
     imageSize=<88 88>
     margin=<ltrb 0 0 0 0>
     outlineSettings {
      bUseBrushTransparency=false
      roundingType=HalfHeightRadius
      width=0
      color {colorUseRule=UseColor_Specified, specifiedColor=<rgba 0 0 0 0>}
      cornerRadii=<0 0 0 0>
     }
     tintColor {colorUseRule=UseColor_Specified, specifiedColor=<rgba 1 1 1 1>}
     uVRegion {bIsValid=false, max=<0 0>, min=<0 0>}
    }}
   slot {class=HorizontalBoxSlot}
  }
  DividerLine {
   widgetClassPath=&Image
   bIsVariable=true
   parent=ModalHeader
   widget=DividerLine
   properties {
    brush {
     bIsDynamicallyLoaded=false
     drawAs=Image
     imageType=NoImage
     mirroring=NoMirror
     resourceName=None
     resourceObject=/Game/Items/Traps/Blueprints/Toys/UI_Art_Assets/M_UI_ActionDividerLine
     tiling=NoTile
     imageSize=<24 32>
     margin=<ltrb 0 0 0 0>
     outlineSettings {
      bUseBrushTransparency=false
      roundingType=HalfHeightRadius
      width=0
      color {colorUseRule=UseColor_Specified, specifiedColor=<rgba 0 0 0 0>}
      cornerRadii=<0 0 0 0>
     }
     tintColor {colorUseRule=UseColor_Specified, specifiedColor=<rgba 1 1 1 1>}
     uVRegion {bIsValid=false, max=<0 0>, min=<0 0>}
    }
    renderTransform {angle=0, scale=<1 1.2>, shear=<-2 0>, translation=<0 2>}
   }
   slot {class=HorizontalBoxSlot, properties {padding=<ltrb -4 0 0 0>}}
  }
  VerticalBox_74 {
   widgetClassPath=&VerticalBox
   parent=ModalHeader
   widget=VerticalBox_74
   properties {renderTransform {angle=0, scale=<1 1>, shear=<-12 0>, translation=<0 0>}}
   slot {class=HorizontalBoxSlot, properties {padding=<ltrb 4 0 0 0>}}
  }
  HeaderColor {
   widgetClassPath=&Border
   bIsVariable=true
   parent=VerticalBox_74
   widget=HeaderColor
   properties {
    background {
     bIsDynamicallyLoaded=false
     drawAs=NoDrawType
     imageType=NoImage
     mirroring=NoMirror
     resourceName=None
     resourceObject=null
     tiling=NoTile
     imageSize=<32 32>
     margin=<ltrb 0 0 0 0>
     outlineSettings {
      bUseBrushTransparency=false
      roundingType=HalfHeightRadius
      width=0
      color {colorUseRule=UseColor_Specified, specifiedColor=<rgba 0 0 0 0>}
      cornerRadii=<0 0 0 0>
     }
     tintColor {colorUseRule=UseColor_Specified, specifiedColor=<rgba 1 1 1 1>}
     uVRegion {bIsValid=false, max=<0 0>, min=<0 0>}
    }
    padding=<ltrb 0 0 0 0>
   }
   slot {class=VerticalBoxSlot}
  }
  HeaderMainText {
   widgetClassPath=&CommonTextBlock
   bIsVariable=true
   parent=HeaderColor
   widget=HeaderMainText
   properties {
    text=Header Text
    textTransformPolicy=ToUpper
    font {
     bForceMonospaced=false
     bMaterialIsStencil=false
     fontMaterial=null
     fontObject=&Burbank
     letterSpacing=0
     monospacedWidth=1
     size=32
     skewAmount=0
     typefaceFontName=BurbankRegular-Black
     outlineSettings {
      bApplyOutlineToDropShadows=false
      bMiteredCorners=false
      bSeparateFillAlpha=false
      outlineMaterial=null
      outlineSize=0
      outlineColor=<rgba 0 0 0 1>
     }
    }
    margin=<ltrb 4 0 4 0>
    shadowOffset=<0 0>
   }
   slot {class=BorderSlot, properties {padding=<ltrb 0 0 0 0>}}
  }
  SubtextColor {
   widgetClassPath=&Border
   bIsVariable=true
   parent=VerticalBox_74
   widget=SubtextColor
   properties {
    background {
     bIsDynamicallyLoaded=false
     drawAs=NoDrawType
     imageType=NoImage
     mirroring=NoMirror
     resourceName=None
     resourceObject=null
     tiling=NoTile
     imageSize=<32 32>
     margin=<ltrb 0 0 0 0>
     outlineSettings {
      bUseBrushTransparency=false
      roundingType=HalfHeightRadius
      width=0
      color {colorUseRule=UseColor_Specified, specifiedColor=<rgba 0 0 0 0>}
      cornerRadii=<0 0 0 0>
     }
     tintColor {colorUseRule=UseColor_Specified, specifiedColor=<rgba 1 1 1 1>}
     uVRegion {bIsValid=false, max=<0 0>, min=<0 0>}
    }
    padding=<ltrb 0 0 0 0>
   }
   slot {class=VerticalBoxSlot}
  }
  SubHeaderText {
   widgetClassPath=&CommonTextBlock
   bIsVariable=true
   parent=SubtextColor
   widget=SubHeaderText
   properties {
    text=SubHeader Text extra text
    textTransformPolicy=ToUpper
    font {
     bForceMonospaced=false
     bMaterialIsStencil=false
     fontMaterial=null
     fontObject=&BurbankSmall
     letterSpacing=0
     monospacedWidth=1
     size=18
     skewAmount=0
     typefaceFontName=Black
     outlineSettings {
      bApplyOutlineToDropShadows=false
      bMiteredCorners=false
      bSeparateFillAlpha=false
      outlineMaterial=null
      outlineSize=0
      outlineColor=<rgba 0 0 0 1>
     }
    }
    margin=<ltrb 4 0 4 0>
    shadowOffset=<0 0>
   }
   slot {class=BorderSlot, properties {padding=<ltrb 0 0 0 0>}}
  }
  VisibilityBorder_CloseTouch {
   widgetClassPath=&CommonHardwareVisibilityBorder
   parent=Overlay_0
   widget=VisibilityBorder_CloseTouch
   properties {
    horizontalAlignment=HAlign_Right
    verticalAlignment=VAlign_Top
    visibilityQuery {
     autoDescription=" ANY( Input.Touch )"
     tokenStreamVersion=0
     userDescription=""
     queryTokenStream [0,1,1,1,0]
     tagDictionary [{tagName=Input.Touch}]
    }
   }
   slot {class=OverlaySlot, properties {horizontalAlignment=HAlign_Right, padding=<ltrb 0 32 32 0>}}
  }
  SafeZone_0 {
   widgetClassPath=&SafeZone
   bIsVariable=true
   parent=VisibilityBorder_CloseTouch
   widget=SafeZone_0
   slot {class=BorderSlot, properties {horizontalAlignment=HAlign_Right, verticalAlignment=VAlign_Top, padding=<ltrb 0 0 0 0>}}
  }
  Button_CloseTouch {
   widgetClassPath=/FortUILibrary/Foundation/Buttons/CloseButton
   bIsVariable=true
   parent=SafeZone_0
   widget=Button_CloseTouch
   properties {
    onButtonBaseClicked=(&self:WidgetTree.Button_CloseTouch.HandleOnButtonClicked)
    onButtonBaseDoubleClicked=(&self:WidgetTree.Button_CloseTouch.HandleOnButtonDoubleClicked)
    onButtonBaseHovered=(&self:WidgetTree.Button_CloseTouch.HandleOnButtonHovered)
    onButtonBaseUnhovered=(&self:WidgetTree.Button_CloseTouch.HandleOnButtonUnhovered)
    onSelectedChangedBase=(&self:WidgetTree.Button_CloseTouch.HandleOnSelectedChanged)
   }
   slot {class=SafeZoneSlot}
  }
 ]
}
