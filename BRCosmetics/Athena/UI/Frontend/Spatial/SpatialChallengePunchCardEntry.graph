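# Widget Blueprint that lays out a spatial frontend challenge punch card entry, derived from the native SpatialChallengePunchCardEntry class.
# A vertical box splits into two equally filled, fixed-height size boxes, each vertically centering a ChallengePunchCard widget as a top and bottom card.
# The two cards are widgets inherited from the parent class, and the event graph holds only empty default events, so this asset contributes layout alone.
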WidgetBlueprint /BRCosmetics/Athena/UI/Frontend/Spatial/SpatialChallengePunchCardEntry
graphs [EventGraph {nodes [
   Event PreConstruct()
   Event Construct()
   Event Tick()
  ]}]
variables []
widgets {
 info {
  inheritedWidgetCount=2
  parentClass=/Script/FortniteUI.SpatialChallengePunchCardEntry
  rootWidgetClass=&VerticalBox
  widgetCount=5
 }
 widgets [
  VerticalBox_23 {widgetClassPath=&VerticalBox, widget=VerticalBox_23}
  SizeBox_2 {
   widgetClassPath=&SizeBox
   parent=VerticalBox_23
   widget=SizeBox_2
   properties {bOverride_HeightOverride=true, heightOverride=300}
   slot {class=VerticalBoxSlot, properties {size {sizeRule=Fill, value=1}}}
  }
  PunchCard_Top {
   widgetClassPath=/Game/Athena/UI/Challenges/ChallengePunchCard
   bInherited=true
   bIsVariable=true
   parent=SizeBox_2
   widget=PunchCard_Top
   slot {class=SizeBoxSlot, properties {verticalAlignment=VAlign_Center}}
  }
  SizeBox_3 {
   widgetClassPath=&SizeBox
   parent=VerticalBox_23
   widget=SizeBox_3
   properties {bOverride_HeightOverride=true, heightOverride=300}
   slot {class=VerticalBoxSlot, properties {size {sizeRule=Fill, value=1}}}
  }
  PunchCard_Bottom {
   widgetClassPath=&ChallengePunchCard
   bInherited=true
   bIsVariable=true
   parent=SizeBox_3
   widget=PunchCard_Bottom
   slot {class=SizeBoxSlot, properties {verticalAlignment=VAlign_Center}}
  }
 ]
}
