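# Season 14 "Nexus War" player level widget, derived from AthenaPlayerLevelDisplay, showing Battle Pass status, level and XP progress.
# A slanted layout pairs the inherited level text with a season title and progress bar image.
# OnUpdateXpBar colors the bar's dynamic material purple and sets its Progress scalar, forcing it full at max level.
# UpdateBPStatus switches between purchased and unpurchased Battle Pass icons; the LevelChanged function is empty.
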WidgetBlueprint /BRCosmetics/Athena/UI/Progression/AthenaPlayerLevelS14
graphs [
 EventGraph {nodes [
   n1 = Event OnUpdateXpBar!() -> (OutputDelegate, then, Progress:float, bIsMaxLevel:bool) {is=title K2Node_Event}
   > n2 = GetDynamicMaterial(self=Get Image_ProgressBar() -> (Image_ProgressBar:Image Object Reference))
   > SetVectorParameterValue(self=[@n2, @n2], ParameterName=ColorA, Value=(R=0.1683,G=0,B=1,A=1))
   > SetVectorParameterValue(self=[@n2, @n2], ParameterName=ColorB, Value=(R=0.2747,G=0,B=1,A=1))
   > n3 = Branch(Condition=@n1.bIsMaxLevel)
   > SetScalarParameterValue(self=@n2, ParameterName=Progress, Value=1.0f)
   SetScalarParameterValue(execute=@n3.else, self=@n2, ParameterName=Progress, Value=@n1.Progress)
   n4 = Event UpdateBPStatus!() -> (OutputDelegate, then, Status:EBPStatus Enum=UnpurchasedBP) {is=title K2Node_Event}
   > SetActiveWidget(
    self=Get Switcher_BPStatus() -> (Switcher_BPStatus:Common Visibility Switcher Object Reference)
    Widget=Select!(
     UnpurchasedBP:Widget Object Reference=Get Image_UnpurchasedBP() -> (Image_UnpurchasedBP:Image Object Reference)
     PurchasedBP:Widget Object Reference=Get Image_PurchasedBP() -> (Image_PurchasedBP:Image Object Reference)
     Index:EBPStatus Enum=@n4.Status
    ) -> (ReturnValue:Widget Object Reference)
   )
  ]}
 LevelChanged {nodes [LevelChanged!() -> (then, Level:int) {is=title K2Node_FunctionEntry}]}
]
variables [LevelChanged {}]
widgets {
 info {inheritedWidgetCount=2, parentClass=/Script/FortniteUI.AthenaPlayerLevelDisplay, rootWidgetClass=&HorizontalBox, widgetCount=9}
 widgets [
  HorizontalBox {widgetClassPath=&HorizontalBox, widget=HorizontalBox}
  Switcher_BPStatus {
   widgetClassPath=&CommonVisibilitySwitcher
   bIsVariable=true
   parent=HorizontalBox
   widget=Switcher_BPStatus
   slot {class=HorizontalBoxSlot, properties {verticalAlignment=VAlign_Center, padding=<ltrb 0 0 -32 0>}}
  }
  Image_UnpurchasedBP {
   widgetClassPath=&Image
   bIsVariable=true
   parent=Switcher_BPStatus
   widget=Image_UnpurchasedBP
   properties {brush {
     bIsDynamicallyLoaded=false
     drawAs=Image
     imageType=NoImage
     mirroring=NoMirror
     resourceName=None
     resourceObject=/BRCosmetics/UI/Foundation/Textures/Icons/Items/T-FNBR-BattlePass-Default
     tiling=NoTile
     imageSize=<64 64>
     margin=<ltrb 0 0 0 0>
     outlineSettings {
      bUseBrushTransparency=false
      roundingType=HalfHeightRadius
      width=0
      color {colorUseRule=UseColor_Specified, specifiedColor=<rgba 0 0 0 0>}
      cornerRadii=<0 0 0 0>
     }
     tintColor {colorUseRule=UseColor_Specified, specifiedColor=<rgba 0 0 0 0.33>}
     uVRegion {bIsValid=false, max=<0 0>, min=<0 0>}
    }}
   slot {class=CommonVisibilitySwitcherSlot}
  }
  Image_PurchasedBP {
   widgetClassPath=&Image
   bIsVariable=true
   parent=Switcher_BPStatus
   widget=Image_PurchasedBP
   properties {brush {
     bIsDynamicallyLoaded=false
     drawAs=Image
     imageType=NoImage
     mirroring=NoMirror
     resourceName=None
     resourceObject=/Game/UI/Foundation/Textures/Icons/Items/T-FNBR-BattlePass
     tiling=NoTile
     imageSize=<64 64>
     margin=<ltrb 0 0 0 0>
     outlineSettings {
      bUseBrushTransparency=false
      roundingType=HalfHeightRadius
      width=0
      color {colorUseRule=UseColor_Specified, specifiedColor=<rgba 0 0 0 0>}
      cornerRadii=<0 0 0 0>
     }
     tintColor {colorUseRule=UseColor_Specified, specifiedColor=<rgba 0 0 0 0.33>}
     uVRegion {bIsValid=false, max=<0 0>, min=<0 0>}
    }}
   slot {class=CommonVisibilitySwitcherSlot}
  }
  Text_Level {
   widgetClassPath=&CommonTextBlock
   bInherited=true
   parent=HorizontalBox
   widget=Text_Level
   properties {
    lineHeightPercentage=0.8
    text=25
    font {
     bForceMonospaced=false
     bMaterialIsStencil=false
     fontMaterial=null
     fontObject=&Burbank
     letterSpacing=0
     monospacedWidth=1
     size=47
     skewAmount=0
     typefaceFontName=BurbankRegular-Black
     outlineSettings {
      bApplyOutlineToDropShadows=false
      bMiteredCorners=false
      bSeparateFillAlpha=false
      outlineMaterial=null
      outlineSize=0
      outlineColor=<rgba 0 0 0 1>
     }
    }
    margin=<ltrb 0 -16 0 0>
    renderTransform {angle=0, scale=<1 1>, shear=<-12 0>, translation=<0 0>}
    shadowOffset=<0 0>
   }
   slot {class=HorizontalBoxSlot, properties {verticalAlignment=VAlign_Center}}
  }
  VerticalBox_180 {
   widgetClassPath=&VerticalBox
   parent=HorizontalBox
   widget=VerticalBox_180
   properties {renderTransform {angle=0, scale=<1 1>, shear=<-12 0>, translation=<0 0>}}
   slot {class=HorizontalBoxSlot, properties {verticalAlignment=VAlign_Center, padding=<ltrb 16 0 0 0>}}
  }
  Text_SeasonTheme {
   widgetClassPath=&CommonTextBlock
   parent=VerticalBox_180
   widget=Text_SeasonTheme
   properties {
    text=NEXUS WAR
    colorAndOpacity {colorUseRule=UseColor_Specified, specifiedColor=<rgba 1 0.6584 0.0356 1>}
    font {
     bForceMonospaced=false
     bMaterialIsStencil=false
     fontMaterial=null
     fontObject=&Burbank
     letterSpacing=0
     monospacedWidth=1
     size=24
     skewAmount=0
     typefaceFontName=BurbankRegular-Black
     outlineSettings {
      bApplyOutlineToDropShadows=false
      bMiteredCorners=false
      bSeparateFillAlpha=false
      outlineMaterial=null
      outlineSize=0
      outlineColor=<rgba 0 0 0 1>
     }
    }
    shadowOffset=<0 0>
   }
   slot {class=VerticalBoxSlot, properties {verticalAlignment=VAlign_Center, size {sizeRule=Fill, value=1}}}
  }
  Image_ProgressBar {
   widgetClassPath=&Image
   bIsVariable=true
   parent=VerticalBox_180
   widget=Image_ProgressBar
   properties {brush {
     bIsDynamicallyLoaded=false
     drawAs=Image
     imageType=NoImage
     mirroring=NoMirror
     resourceName=None
     resourceObject=/BRCosmetics/Athena/UI/Challenges/Art/T_Ui_Spatial_BP_Progress_Inst
     tiling=NoTile
     imageSize=<384 16>
     margin=<ltrb 0 0 0 0>
     outlineSettings {
      bUseBrushTransparency=false
      roundingType=HalfHeightRadius
      width=0
      color {colorUseRule=UseColor_Specified, specifiedColor=<rgba 0 0 0 0>}
      cornerRadii=<0 0 0 0>
     }
     tintColor {colorUseRule=UseColor_Specified, specifiedColor=<rgba 1 1 1 1>}
     uVRegion {bIsValid=false, max=<0 0>, min=<0 0>}
    }}
   slot {class=VerticalBoxSlot, properties {horizontalAlignment=HAlign_Left}}
  }
  RewardIcon {widgetClassPath=/Script/FortniteUI.AthenaSeasonItemRewardIcon, bInherited=true, widget=null}
 ]
}
