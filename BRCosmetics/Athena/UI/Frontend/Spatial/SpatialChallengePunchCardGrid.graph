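# Spatial Challenge Punch Card Grid screen for the frontend, showing the season's punch cards in a horizontal scroll box under a heading.
# SetDiscoveredCount writes a discovered count, or a completed count once all are known, and toggles the description text.
# The About button fires an analytic and switches to BPS12_AboutModal with an animation, as does OnShowFirstTimeInfo; closing returns to the grid.
# PreConstruct tints the two info-card backgrounds through dynamic materials.
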WidgetBlueprint /BRCosmetics/Athena/UI/Frontend/Spatial/SpatialChallengePunchCardGrid
animations [
 intro {
  binding_count=3
  length_seconds=0.5
  object=&self:intro
  sequence {bindings [
    Overlay_0 {
     id=#1
     possessed_class=Overlay
     tracks [{
       type=MovieSceneFloatTrack
       display_name=Render Opacity
       sections [{type=MovieSceneFloatSection, channels [None {type=MovieSceneScriptingFloatChannel, keys [[0,0],[6,1]]}]}]
      }]
    }
    SeasonHeadingTextLabel {
     id=#2
     possessed_class=CommonTextBlock
     tracks [
      {
       type=MovieScene2DTransformTrack
       sections [{
         type=MovieScene2DTransformSection
         channels [
          Translation.X {type=MovieSceneScriptingFloatChannel, keys [[4,128],[8,0]]}
          Translation.Y {type=MovieSceneScriptingFloatChannel}
          Angle {type=MovieSceneScriptingFloatChannel}
          Scale.X {type=MovieSceneScriptingFloatChannel}
          Scale.Y {type=MovieSceneScriptingFloatChannel}
          Shear.X {type=MovieSceneScriptingFloatChannel}
          Shear.Y {type=MovieSceneScriptingFloatChannel}
         ]
        }]
      }
      {
       type=MovieSceneFloatTrack
       display_name=Render Opacity
       sections [{type=MovieSceneFloatSection, channels [None {type=MovieSceneScriptingFloatChannel, keys [[4,0],[8,1]]}]}]
      }
     ]
    }
    RichText_DIscovered {
     id=#3
     possessed_class=CommonRichTextBlock
     tracks [
      {
       type=MovieScene2DTransformTrack
       sections [{
         type=MovieScene2DTransformSection
         channels [
          Translation.X {type=MovieSceneScriptingFloatChannel, keys [[6,128],[10,0]]}
          Translation.Y {type=MovieSceneScriptingFloatChannel}
          Angle {type=MovieSceneScriptingFloatChannel}
          Scale.X {type=MovieSceneScriptingFloatChannel}
          Scale.Y {type=MovieSceneScriptingFloatChannel}
          Shear.X {type=MovieSceneScriptingFloatChannel}
          Shear.Y {type=MovieSceneScriptingFloatChannel}
         ]
        }]
      }
      {
       type=MovieSceneFloatTrack
       display_name=Render Opacity
       sections [{type=MovieSceneFloatSection, channels [None {type=MovieSceneScriptingFloatChannel, keys [[6,0],[10,1]]}]}]
      }
     ]
    }
   ]}
 }
 intro_grid {
  binding_count=5
  length_seconds=0.3501
  object=&self:intro_grid
  sequence {bindings [
    Border_0 {
     id=#4
     possessed_class=Border
     tracks [{
       type=MovieScene2DTransformTrack
       sections [{
         type=MovieScene2DTransformSection
         channels [
          Translation.X {type=MovieSceneScriptingFloatChannel}
          Translation.Y {type=MovieSceneScriptingFloatChannel}
          Angle {type=MovieSceneScriptingFloatChannel}
          Scale.X {type=MovieSceneScriptingFloatChannel}
          Scale.Y {type=MovieSceneScriptingFloatChannel, keys [[2.9993,0],[9.9993,1]]}
          Shear.X {type=MovieSceneScriptingFloatChannel}
          Shear.Y {type=MovieSceneScriptingFloatChannel}
         ]
        }]
      }]
    }
    Button_TouchBack {
     id=#5
     possessed_class=CloseButton_C
     tracks [
      {
       type=MovieScene2DTransformTrack
       sections [{
         type=MovieScene2DTransformSection
         channels [
          Translation.X {type=MovieSceneScriptingFloatChannel, keys [[2.9993,256],[8,0]]}
          Translation.Y {type=MovieSceneScriptingFloatChannel}
          Angle {type=MovieSceneScriptingFloatChannel}
          Scale.X {type=MovieSceneScriptingFloatChannel}
          Scale.Y {type=MovieSceneScriptingFloatChannel}
          Shear.X {type=MovieSceneScriptingFloatChannel}
          Shear.Y {type=MovieSceneScriptingFloatChannel}
         ]
        }]
      }
      {
       type=MovieSceneFloatTrack
       display_name=Render Opacity
       sections [{type=MovieSceneFloatSection, channels [None {type=MovieSceneScriptingFloatChannel, keys [[3,0],[8,1]]}]}]
      }
     ]
    }
    Button_Back {
     id=#6
     possessed_class=IconTextButton_C
     tracks [
      {
       type=MovieScene2DTransformTrack
       sections [{
         type=MovieScene2DTransformSection
         channels [
          Translation.X {type=MovieSceneScriptingFloatChannel, keys [[5,256],[10,0]]}
          Translation.Y {type=MovieSceneScriptingFloatChannel}
          Angle {type=MovieSceneScriptingFloatChannel}
          Scale.X {type=MovieSceneScriptingFloatChannel}
          Scale.Y {type=MovieSceneScriptingFloatChannel}
          Shear.X {type=MovieSceneScriptingFloatChannel}
          Shear.Y {type=MovieSceneScriptingFloatChannel}
         ]
        }]
      }
      {
       type=MovieSceneFloatTrack
       display_name=Render Opacity
       sections [{type=MovieSceneFloatSection, channels [None {type=MovieSceneScriptingFloatChannel, keys [[5,0],[10,1]]}]}]
      }
     ]
    }
    Button_About {
     id=#7
     possessed_class=IconTextButton_C
     tracks [
      {
       type=MovieScene2DTransformTrack
       sections [{
         type=MovieScene2DTransformSection
         channels [
          Translation.X {type=MovieSceneScriptingFloatChannel, keys [[2.9993,256],[7.9993,0]]}
          Translation.Y {type=MovieSceneScriptingFloatChannel}
          Angle {type=MovieSceneScriptingFloatChannel}
          Scale.X {type=MovieSceneScriptingFloatChannel}
          Scale.Y {type=MovieSceneScriptingFloatChannel}
          Shear.X {type=MovieSceneScriptingFloatChannel}
          Shear.Y {type=MovieSceneScriptingFloatChannel}
         ]
        }]
      }
      {
       type=MovieSceneFloatTrack
       display_name=Render Opacity
       sections [{type=MovieSceneFloatSection, channels [None {type=MovieSceneScriptingFloatChannel, keys [[2.9993,0],[7.9993,1]]}]}]
      }
     ]
    }
    CommonRichTextBlock_Description {
     id=#8
     possessed_class=CommonRichTextBlock
     tracks [
      {
       type=MovieScene2DTransformTrack
       sections [{
         type=MovieScene2DTransformSection
         channels [
          Translation.X {type=MovieSceneScriptingFloatChannel, keys [[4,-256],[9.0007,0]]}
          Translation.Y {type=MovieSceneScriptingFloatChannel}
          Angle {type=MovieSceneScriptingFloatChannel}
          Scale.X {type=MovieSceneScriptingFloatChannel}
          Scale.Y {type=MovieSceneScriptingFloatChannel}
          Shear.X {type=MovieSceneScriptingFloatChannel}
          Shear.Y {type=MovieSceneScriptingFloatChannel}
         ]
        }]
      }
      {
       type=MovieSceneFloatTrack
       display_name=Render Opacity
       sections [{type=MovieSceneFloatSection, channels [None {type=MovieSceneScriptingFloatChannel, keys [[4,0],[9.0007,1]]}]}]
      }
     ]
    }
   ]}
 }
 intro_about {
  binding_count=3
  length_seconds=0.35
  object=&self:intro_about
  sequence {bindings [
    SizeBox_Card2 {
     id=#9
     possessed_class=SizeBox
     tracks [{
       type=MovieScene2DTransformTrack
       sections [{
         type=MovieScene2DTransformSection
         channels [
          Translation.X {type=MovieSceneScriptingFloatChannel, keys [[3,52.9383],[7,20.8999]]}
          Translation.Y {type=MovieSceneScriptingFloatChannel, keys [[3,77.9454],[7,33.1247]]}
          Angle {type=MovieSceneScriptingFloatChannel, keys [[3,2],[7,13]]}
          Scale.X {type=MovieSceneScriptingFloatChannel}
          Scale.Y {type=MovieSceneScriptingFloatChannel}
          Shear.X {type=MovieSceneScriptingFloatChannel}
          Shear.Y {type=MovieSceneScriptingFloatChannel}
         ]
        }]
      }]
    }
    SizeBox_Card1 {
     id=#10
     possessed_class=SizeBox
     tracks [{
       type=MovieScene2DTransformTrack
       sections [{
         type=MovieScene2DTransformSection
         channels [
          Translation.X {type=MovieSceneScriptingFloatChannel, keys [[3,-120.9421],[7,-79]]}
          Translation.Y {type=MovieSceneScriptingFloatChannel, keys [[3,-250.0356],[7,-166]]}
          Angle {type=MovieSceneScriptingFloatChannel, keys [[3,-5],[7,-13]]}
          Scale.X {type=MovieSceneScriptingFloatChannel}
          Scale.Y {type=MovieSceneScriptingFloatChannel}
          Shear.X {type=MovieSceneScriptingFloatChannel}
          Shear.Y {type=MovieSceneScriptingFloatChannel}
         ]
        }]
      }]
    }
    Overlay_InfoCards {
     id=#11
     possessed_class=Overlay
     tracks [{
       type=MovieSceneFloatTrack
       display_name=Render Opacity
       sections [{type=MovieSceneFloatSection, channels [None {type=MovieSceneScriptingFloatChannel, keys [[0.0003,0],[4,0],[7,1]]}]}]
      }]
    }
   ]}
 }
]
graphs [EventGraph {nodes [
   "On Button Clicked (Button_About)"!() -> (OutputDelegate, then, Button:Common Button Legacy Object Reference) {is=title K2Node_ComponentBoundEvent}
   > n1 = FireSpatialScreenAnalytic!(execute, self:Athena Spatial Screen Object Reference, Interaction:String=MoreInfo) -> (then)
   SetActiveWidget(
    execute=[@n1.then, @n4.then]
    self:Widget Switcher Object Reference=Get CommonActivatableWidgetSwitcher_State() -> (CommonActivatableWidgetSwitcher_State:Common Activatable Widget Switcher Object Reference)
    Widget=Get BPS12_AboutModal() -> (BPS12_AboutModal:BPS12 About Modal Object Reference)
   )
   > PlayAnimationForward(InAnimation=Get intro_about() -> (intro_about:Widget Animation Object Reference))
   "Close About (BPS12_AboutModal)"() {is=title K2Node_ComponentBoundEvent}
   > PlayAnimationReverse(InAnimation=Get intro_about())
   > SetActiveWidget(self=Get CommonActivatableWidgetSwitcher_State(), Widget=Get SafeZone() -> (SafeZone:Safe Zone Object Reference))
   > PlayAnimationForward(InAnimation=Get intro_grid() -> (intro_grid:Widget Animation Object Reference))
   IntroFinished() {is=title K2Node_CustomEvent}
   n2 = Get FocusedPunchCard() -> (FocusedPunchCard:Athena Challenge Punch Card Weak object)
   > Is Valid(InputObject=@n2)
   > n3 = Cast To FortPlayerController(Object=GetOwningPlayer())
   > Branch(Condition=IsUsingGamepad!(self:Fort Player Controller Object Reference=@n3) -> (ReturnValue:bool))
   > SetFocus(self=@n2)
   n4 = Event OnShowFirstTimeInfo() {is=title K2Node_Event}
   Event Construct()
   > BindToAnimationFinished(Animation=Get intro_grid(), Delegate=Create Event())
   n5 = Event SetDiscoveredCount!() -> (OutputDelegate, then, KnownCount:int=0, CompletedCount:int=0, TotalCount:int=0) {is=title K2Node_Event}
   > SetVisibility(
    self=Get CommonRichTextBlock_Description() -> (CommonRichTextBlock_Description:Common Rich Text Block Object Reference)
    InVisibility=Select(Option 0=Hidden, Option 1=HitTestInvisible, Index="Equal (Integer)"(A=@n5.KnownCount, B=0) {is=title K2Node_CallFunction})
   )
   > n6 = Branch(Condition="integer < integer"!(A:int=@n5.KnownCount, B:int=@n5.TotalCount) -> (ReturnValue:bool) {is=title K2Node_CallFunction})
   n7 = Get RichText_DIscovered() -> (RichText_DIscovered:Common Rich Text Block Object Reference)
   > SetText(
    self=@n7
    InText=Format Text!(Format="LOCTEXT(\"<text color=\\\"fff\\\">{Known} of {Total}</> DISCOVERED\")", Known:int=@n5.KnownCount, Total:int=@n5.TotalCount) -> (Result)
   )
   SetText(
    execute=@n6.else
    self=@n7
    InText=Format Text!(Format="LOCTEXT(\"<text color=\\\"fff\\\">{Completed} of {Total}</> COMPLETED\")", Completed:int=@n5.CompletedCount, Total=@n5.TotalCount) -> (Result)
   )
   Event On Activated()
   > PlayAnimationForward(InAnimation=Get intro() -> (intro:Widget Animation Object Reference))
   > PlayAnimationForward(InAnimation=Get intro_grid())
   Event PreConstruct()
   > n8 = Sequence()
   > n9 = GetDynamicMaterial(self=Get FortLazyImage_Background_01() -> (FortLazyImage_Background_01:Fort Lazy Image Object Reference))
   > SetVectorParameterValue(self=@n9, ParameterName=BaseColor02, Value=Get card1_base2() -> (card1_base2:Linear Color Structure))
   > SetVectorParameterValue(self=@n9, ParameterName=BaseColor03, Value=Get card1_base3() -> (card1_base3:Linear Color Structure))
   > SetVectorParameterValue(self=@n9, ParameterName=BaseColor04, Value=Get Card1_base4() -> (Card1_base4:Linear Color Structure))
   > SetTextureParameterValue(self=@n9, ParameterName=Background, Value=/Game/Athena/UI/Challenges/Art/PunchCard_Icons/CategoriesTextures/T_UI_punchcard_None)
   > n10 = GetDynamicMaterial(self=Get Image_EmptySocket_01_left() -> (Image_EmptySocket_01_left:Image Object Reference))
   > SetScalarParameterValue(self=@n10, ParameterName=ShowLeft, Value=0)
   > n11 = GetDynamicMaterial(self=Get Image_EmptySocket_01_Right() -> (Image_EmptySocket_01_Right:Image Object Reference))
   > SetScalarParameterValue(self=@n11, ParameterName=ShowRight, Value=0)
   n12 = GetDynamicMaterial(execute=@n8.then_1, self=Get FortLazyImage_Background_02() -> (FortLazyImage_Background_02:Fort Lazy Image Object Reference))
   > SetVectorParameterValue(self=@n12, ParameterName=BaseColor02, Value=Get card2_base2() -> (card2_base2:Linear Color Structure))
   > SetVectorParameterValue(self=@n12, ParameterName=BaseColor03, Value=Get card2_base3() -> (card2_base3:Linear Color Structure))
   > SetVectorParameterValue(self=@n12, ParameterName=BaseColor04, Value=Get Card2_base4() -> (Card2_base4:Linear Color Structure))
   > SetTextureParameterValue(self=@n12, ParameterName=Background, Value=&T_UI_punchcard_None)
   > n13 = GetDynamicMaterial(self=Get Image_EmptySocket_02_left() -> (Image_EmptySocket_02_left:Image Object Reference))
   > SetScalarParameterValue(self=@n13, ParameterName=ShowLeft, Value=0)
   > n14 = GetDynamicMaterial(self=Get Image_EmptySocket_02_Right() -> (Image_EmptySocket_02_Right:Image Object Reference))
   > SetScalarParameterValue(self=@n14, ParameterName=ShowRight, Value=0)
  ]}]
variables [
 card1_base2 {type=LinearColor, default=<rgba 1 0.004 0.0307 1>}
 card1_base3 {type=LinearColor, default=<rgba 0.3467 0.0015 0.013 1>}
 Card1_base4 {type=LinearColor, default=<rgba 0.1354 0 0.015 1>}
 card2_base2 {type=LinearColor, default=<rgba 1 0.004 0.0307 1>}
 card2_base3 {type=LinearColor, default=<rgba 0.3467 0.0015 0.013 1>}
 Card2_base4 {type=LinearColor, default=<rgba 0.1354 0 0.015 1>}
]
widgets {
 info {inheritedWidgetCount=4, parentClass=/Script/FortniteUI.SpatialChallengePunchCard, rootWidgetClass=&Overlay, widgetCount=64}
 widgets [
  Overlay_0 {widgetClassPath=&Overlay, widget=Overlay_0}
  BackgroundGradient {
   widgetClassPath=&Image
   parent=Overlay_0
   widget=BackgroundGradient
   properties {
    visibility=SelfHitTestInvisible
    brush {
     bIsDynamicallyLoaded=false
     drawAs=Image
     imageType=NoImage
     mirroring=NoMirror
     resourceName=None
     resourceObject=/BRCosmetics/Athena/UI/Challenges/Art/M_UI_S14_PC_Backgrund_Inst
     tiling=NoTile
     imageSize=<1920 1080>
     margin=<ltrb 0 0 0 0>
     outlineSettings {
      bUseBrushTransparency=false
      roundingType=HalfHeightRadius
      width=0
      color {colorUseRule=UseColor_Specified, specifiedColor=<rgba 0 0 0 0>}
      cornerRadii=<0 0 0 0>
     }
     tintColor {colorUseRule=UseColor_Specified, specifiedColor=<rgba 1 1 1 1>}
     uVRegion {bIsValid=false, max=<0 0>, min=<0 0>}
    }
   }
   slot {class=OverlaySlot, properties {horizontalAlignment=HAlign_Fill, verticalAlignment=VAlign_Fill}}
  }
  ScanlineOverlay {
   widgetClassPath=&Image
   bIsVariable=true
   parent=Overlay_0
   widget=ScanlineOverlay
   properties {
    visibility=Collapsed
    brush {
     bIsDynamicallyLoaded=false
     drawAs=Image
     imageType=NoImage
     mirroring=NoMirror
     resourceName=None
     resourceObject=/FortniteCoreUI/Frontend/Art/StreamLoading/M_UI_TimeWaster_ScanlineOverlay_Inst_OldPC
     tiling=NoTile
     imageSize=<32 32>
     margin=<ltrb 0 0 0 0>
     outlineSettings {
      bUseBrushTransparency=false
      roundingType=HalfHeightRadius
      width=0
      color {colorUseRule=UseColor_Specified, specifiedColor=<rgba 0 0 0 0>}
      cornerRadii=<0 0 0 0>
     }
     tintColor {colorUseRule=UseColor_Specified, specifiedColor=<rgba 1 1 1 1>}
     uVRegion {bIsValid=false, max=<0 0>, min=<0 0>}
    }
    colorAndOpacity=<rgba 1 0 0.0296 0.0235>
    renderTransform {angle=0, scale=<1 -1>, shear=<0 0>, translation=<0 0>}
   }
   slot {class=OverlaySlot, properties {horizontalAlignment=HAlign_Fill, verticalAlignment=VAlign_Fill, padding=<ltrb -1 -1 -1 -1>}}
  }
  AthenaPlayerLevelS14 {
   widgetClassPath=/BRCosmetics/Athena/UI/Progression/AthenaPlayerLevelS14
   bIsVariable=true
   parent=Overlay_0
   widget=AthenaPlayerLevelS14
   slot {class=OverlaySlot, properties {verticalAlignment=VAlign_Bottom, padding=<ltrb 32 32 32 32>}}
  }
  VerticalBox_1 {
   widgetClassPath=&VerticalBox
   parent=Overlay_0
   widget=VerticalBox_1
   slot {class=OverlaySlot, properties {padding=<ltrb 32 32 32 32>}}
  }
  SeasonHeadingTextLabel {
   widgetClassPath=&CommonTextBlock
   bIsVariable=true
   parent=VerticalBox_1
   widget=SeasonHeadingTextLabel
   properties {
    text=Punch cards
    textTransformPolicy=ToUpper
    colorAndOpacity {colorUseRule=UseColor_Specified, specifiedColor=<rgba 1 0 0.0395 1>}
    font {
     bForceMonospaced=false
     bMaterialIsStencil=false
     fontMaterial=/FortniteCoreUI/UI/Season/Art/M_UI_AthenaSeason_BattlePass_Text_white
     fontObject=&Burbank
     letterSpacing=0
     monospacedWidth=1
     size=96
     skewAmount=0
     typefaceFontName=BurbankRegular-Black
     outlineSettings {
      bApplyOutlineToDropShadows=false
      bMiteredCorners=false
      bSeparateFillAlpha=false
      outlineMaterial=null
      outlineSize=0
      outlineColor=<rgba 0 0 0 1>
     }
    }
    margin=<ltrb 0 -32 0 -32>
    renderTransform {angle=0, scale=<1 1>, shear=<-12 0>, translation=<0 0>}
    renderTransformPivot=<0 0.5>
    shadowOffset=<0 0>
   }
   slot {class=VerticalBoxSlot}
  }
  RichText_DIscovered {
   widgetClassPath=&CommonRichTextBlock
   bIsVariable=true
   parent=VerticalBox_1
   widget=RichText_DIscovered
   properties {
    bOverrideDefaultStyle=true
    text="<text color=\"fff\">26 of 52</> DISCOVERED"
    textTransformPolicy=ToUpper
    defaultTextStyleOverride {
     overflowPolicy=Clip
     transformPolicy=None
     colorAndOpacity {colorUseRule=UseColor_Specified, specifiedColor=<rgba 1 0.6584 0.0356 1>}
     font {
      bForceMonospaced=false
      bMaterialIsStencil=false
      fontMaterial=null
      fontObject=&BurbankSmall
      letterSpacing=0
      monospacedWidth=1
      size=32
      skewAmount=0
      typefaceFontName=Black
      outlineSettings {
       bApplyOutlineToDropShadows=false
       bMiteredCorners=false
       bSeparateFillAlpha=false
       outlineMaterial=null
       outlineSize=0
       outlineColor=<rgba 0 0 0 1>
      }
     }
     highlightColor {colorUseRule=UseColor_Specified, specifiedColor=<rgba 0 0 0 1>}
     highlightShape {
      bIsDynamicallyLoaded=false
      drawAs=Image
      imageType=NoImage
      mirroring=NoMirror
      resourceName=None
      resourceObject=null
      tiling=NoTile
      imageSize=<32 32>
      margin=<ltrb 0 0 0 0>
      outlineSettings {
       bUseBrushTransparency=false
       roundingType=HalfHeightRadius
       width=0
       color {colorUseRule=UseColor_Specified, specifiedColor=<rgba 0 0 0 0>}
       cornerRadii=<0 0 0 0>
      }
      tintColor {colorUseRule=UseColor_Specified, specifiedColor=<rgba 1 1 1 1>}
      uVRegion {bIsValid=false, max=<0 0>, min=<0 0>}
     }
     selectedBackgroundColor {colorUseRule=UseColor_Foreground, specifiedColor=<rgba 1 0 1 1>}
     shadowColorAndOpacity=<rgba 0 0 0 1>
     shadowOffset=<0 0>
     strikeBrush {
      bIsDynamicallyLoaded=false
      drawAs=Image
      imageType=NoImage
      mirroring=NoMirror
      resourceName=None
      resourceObject=null
      tiling=NoTile
      imageSize=<32 32>
      margin=<ltrb 0 0 0 0>
      outlineSettings {
       bUseBrushTransparency=false
       roundingType=HalfHeightRadius
       width=0
       color {colorUseRule=UseColor_Specified, specifiedColor=<rgba 0 0 0 0>}
       cornerRadii=<0 0 0 0>
      }
      tintColor {colorUseRule=UseColor_Specified, specifiedColor=<rgba 1 1 1 1>}
      uVRegion {bIsValid=false, max=<0 0>, min=<0 0>}
     }
     underlineBrush {
      bIsDynamicallyLoaded=false
      drawAs=Image
      imageType=NoImage
      mirroring=NoMirror
      resourceName=None
      resourceObject=null
      tiling=NoTile
      imageSize=<32 32>
      margin=<ltrb 0 0 0 0>
      outlineSettings {
       bUseBrushTransparency=false
       roundingType=HalfHeightRadius
       width=0
       color {colorUseRule=UseColor_Specified, specifiedColor=<rgba 0 0 0 0>}
       cornerRadii=<0 0 0 0>
      }
      tintColor {colorUseRule=UseColor_Specified, specifiedColor=<rgba 1 1 1 1>}
      uVRegion {bIsValid=false, max=<0 0>, min=<0 0>}
     }
    }
    renderTransform {angle=0, scale=<1 1>, shear=<-12 0>, translation=<0 0>}
   }
   slot {class=VerticalBoxSlot, properties {horizontalAlignment=HAlign_Left, padding=<ltrb 50 -32 0 0>}}
  }
  CommonActivatableWidgetSwitcher_State {
   widgetClassPath=&CommonActivatableWidgetSwitcher
   bIsVariable=true
   parent=Overlay_0
   widget=CommonActivatableWidgetSwitcher_State
   slot {class=OverlaySlot, properties {horizontalAlignment=HAlign_Fill, verticalAlignment=VAlign_Fill}}
  }
  SafeZone {
   widgetClassPath=&SafeZone
   bIsVariable=true
   parent=CommonActivatableWidgetSwitcher_State
   widget=SafeZone
   slot {class=WidgetSwitcherSlot}
  }
  Overlay_grid {widgetClassPath=&Overlay, bIsVariable=true, parent=SafeZone, widget=Overlay_grid, slot {class=SafeZoneSlot}}
  HorizontalBox_3 {
   widgetClassPath=&HorizontalBox
   parent=Overlay_grid
   widget=HorizontalBox_3
   slot {class=OverlaySlot, properties {horizontalAlignment=HAlign_Right, padding=<ltrb 0 32 32 32>}}
  }
  CommonRichTextBlock_Description {
   widgetClassPath=&CommonRichTextBlock
   bIsVariable=true
   parent=HorizontalBox_3
   widget=CommonRichTextBlock_Description
   properties {
    autoWrapText=true
    bOverrideDefaultStyle=true
    justification=Right
    lineHeightPercentage=0.95
    text="Experiment on the island to \r\n<text color=\"FFFFFFFF\" fontface=\"black\" case=\"upper\">discover punch cards</>."
    defaultTextStyleOverride {
     overflowPolicy=Clip
     transformPolicy=None
     colorAndOpacity {colorUseRule=UseColor_Specified, specifiedColor=<rgba 1 0.6584 0.0356 1>}
     font {
      bForceMonospaced=false
      bMaterialIsStencil=false
      fontMaterial=null
      fontObject=&BurbankSmall
      letterSpacing=0
      monospacedWidth=1
      size=18
      skewAmount=0
      typefaceFontName=Bold
      outlineSettings {
       bApplyOutlineToDropShadows=false
       bMiteredCorners=false
       bSeparateFillAlpha=false
       outlineMaterial=null
       outlineSize=0
       outlineColor=<rgba 0 0 0 1>
      }
     }
     highlightColor {colorUseRule=UseColor_Specified, specifiedColor=<rgba 0 0 0 1>}
     highlightShape {
      bIsDynamicallyLoaded=false
      drawAs=Image
      imageType=NoImage
      mirroring=NoMirror
      resourceName=None
      resourceObject=null
      tiling=NoTile
      imageSize=<32 32>
      margin=<ltrb 0 0 0 0>
      outlineSettings {
       bUseBrushTransparency=false
       roundingType=HalfHeightRadius
       width=0
       color {colorUseRule=UseColor_Specified, specifiedColor=<rgba 0 0 0 0>}
       cornerRadii=<0 0 0 0>
      }
      tintColor {colorUseRule=UseColor_Specified, specifiedColor=<rgba 1 1 1 1>}
      uVRegion {bIsValid=false, max=<0 0>, min=<0 0>}
     }
     selectedBackgroundColor {colorUseRule=UseColor_Foreground, specifiedColor=<rgba 1 0 1 1>}
     shadowColorAndOpacity=<rgba 0 0 0 1>
     shadowOffset=<0 0>
     strikeBrush {
      bIsDynamicallyLoaded=false
      drawAs=Image
      imageType=NoImage
      mirroring=NoMirror
      resourceName=None
      resourceObject=null
      tiling=NoTile
      imageSize=<32 32>
      margin=<ltrb 0 0 0 0>
      outlineSettings {
       bUseBrushTransparency=false
       roundingType=HalfHeightRadius
       width=0
       color {colorUseRule=UseColor_Specified, specifiedColor=<rgba 0 0 0 0>}
       cornerRadii=<0 0 0 0>
      }
      tintColor {colorUseRule=UseColor_Specified, specifiedColor=<rgba 1 1 1 1>}
      uVRegion {bIsValid=false, max=<0 0>, min=<0 0>}
     }
     underlineBrush {
      bIsDynamicallyLoaded=false
      drawAs=Image
      imageType=NoImage
      mirroring=NoMirror
      resourceName=None
      resourceObject=null
      tiling=NoTile
      imageSize=<32 32>
      margin=<ltrb 0 0 0 0>
      outlineSettings {
       bUseBrushTransparency=false
       roundingType=HalfHeightRadius
       width=0
       color {colorUseRule=UseColor_Specified, specifiedColor=<rgba 0 0 0 0>}
       cornerRadii=<0 0 0 0>
      }
      tintColor {colorUseRule=UseColor_Specified, specifiedColor=<rgba 1 1 1 1>}
      uVRegion {bIsValid=false, max=<0 0>, min=<0 0>}
     }
    }
    renderTransform {angle=0, scale=<1 1>, shear=<-12 0>, translation=<0 0>}
   }
   slot {class=HorizontalBoxSlot, properties {verticalAlignment=VAlign_Center, padding=<ltrb 32 0 32 0>, size {sizeRule=Fill, value=1}}}
  }
  HardwareVisibilityBorder_TouchBack {
   widgetClassPath=&CommonHardwareVisibilityBorder
   parent=HorizontalBox_3
   widget=HardwareVisibilityBorder_TouchBack
   properties {visibilityQuery {
     autoDescription=" ANY( Input.Touch )"
     tokenStreamVersion=0
     userDescription=""
     queryTokenStream [0,1,1,1,0]
     tagDictionary [{tagName=Input.Touch}]
    }}
   slot {class=HorizontalBoxSlot}
  }
  Button_TouchBack {
   widgetClassPath=/FortUILibrary/Foundation/Buttons/CloseButton
   bInherited=true
   bIsVariable=true
   parent=HardwareVisibilityBorder_TouchBack
   widget=Button_TouchBack
   properties {
    onButtonBaseClicked=(&self:WidgetTree.Button_TouchBack.HandleOnButtonClicked)
    onButtonBaseDoubleClicked=(&self:WidgetTree.Button_TouchBack.HandleOnButtonDoubleClicked)
    onButtonBaseHovered=(&self:WidgetTree.Button_TouchBack.HandleOnButtonHovered)
    onButtonBaseUnhovered=(&self:WidgetTree.Button_TouchBack.HandleOnButtonUnhovered)
    onSelectedChangedBase=(&self:WidgetTree.Button_TouchBack.HandleOnSelectedChanged)
   }
   slot {class=BorderSlot, properties {padding=<ltrb 0 0 0 0>}}
  }
  VerticalBox_0 {
   widgetClassPath=&VerticalBox
   parent=Overlay_grid
   widget=VerticalBox_0
   slot {class=OverlaySlot, properties {horizontalAlignment=HAlign_Fill, verticalAlignment=VAlign_Fill}}
  }
  HorizontalBox_89 {
   widgetClassPath=&HorizontalBox
   parent=VerticalBox_0
   widget=HorizontalBox_89
   properties {visibility=Collapsed}
   slot {class=VerticalBoxSlot, properties {horizontalAlignment=HAlign_Center}}
  }
  Image_70 {
   widgetClassPath=&Image
   bIsVariable=true
   parent=HorizontalBox_89
   widget=Image_70
   properties {brush {
     bIsDynamicallyLoaded=false
     drawAs=Image
     imageType=NoImage
     mirroring=NoMirror
     resourceName=None
     resourceObject=/Game/UI/Fort_Traits/Circle
     tiling=NoTile
     imageSize=<128 128>
     margin=<ltrb 0 0 0 0>
     outlineSettings {
      bUseBrushTransparency=false
      roundingType=HalfHeightRadius
      width=0
      color {colorUseRule=UseColor_Specified, specifiedColor=<rgba 0 0 0 0>}
      cornerRadii=<0 0 0 0>
     }
     tintColor {colorUseRule=UseColor_Specified, specifiedColor=<rgba 1 1 1 1>}
     uVRegion {bIsValid=false, max=<0 0>, min=<0 0>}
    }}
   slot {class=HorizontalBoxSlot, properties {padding=<ltrb 15 0 15 0>}}
  }
  Image_180 {
   widgetClassPath=&Image
   bIsVariable=true
   parent=HorizontalBox_89
   widget=Image_180
   properties {brush {
     bIsDynamicallyLoaded=false
     drawAs=Image
     imageType=NoImage
     mirroring=NoMirror
     resourceName=None
     resourceObject=&Circle
     tiling=NoTile
     imageSize=<128 128>
     margin=<ltrb 0 0 0 0>
     outlineSettings {
      bUseBrushTransparency=false
      roundingType=HalfHeightRadius
      width=0
      color {colorUseRule=UseColor_Specified, specifiedColor=<rgba 0 0 0 0>}
      cornerRadii=<0 0 0 0>
     }
     tintColor {colorUseRule=UseColor_Specified, specifiedColor=<rgba 1 1 1 1>}
     uVRegion {bIsValid=false, max=<0 0>, min=<0 0>}
    }}
   slot {class=HorizontalBoxSlot, properties {padding=<ltrb 15 0 15 0>}}
  }
  Image_232 {
   widgetClassPath=&Image
   bIsVariable=true
   parent=HorizontalBox_89
   widget=Image_232
   properties {brush {
     bIsDynamicallyLoaded=false
     drawAs=Image
     imageType=NoImage
     mirroring=NoMirror
     resourceName=None
     resourceObject=&Circle
     tiling=NoTile
     imageSize=<128 128>
     margin=<ltrb 0 0 0 0>
     outlineSettings {
      bUseBrushTransparency=false
      roundingType=HalfHeightRadius
      width=0
      color {colorUseRule=UseColor_Specified, specifiedColor=<rgba 0 0 0 0>}
      cornerRadii=<0 0 0 0>
     }
     tintColor {colorUseRule=UseColor_Specified, specifiedColor=<rgba 1 1 1 1>}
     uVRegion {bIsValid=false, max=<0 0>, min=<0 0>}
    }}
   slot {class=HorizontalBoxSlot, properties {padding=<ltrb 15 0 15 0>}}
  }
  Image_287 {
   widgetClassPath=&Image
   bIsVariable=true
   parent=HorizontalBox_89
   widget=Image_287
   properties {brush {
     bIsDynamicallyLoaded=false
     drawAs=Image
     imageType=NoImage
     mirroring=NoMirror
     resourceName=None
     resourceObject=&Circle
     tiling=NoTile
     imageSize=<128 128>
     margin=<ltrb 0 0 0 0>
     outlineSettings {
      bUseBrushTransparency=false
      roundingType=HalfHeightRadius
      width=0
      color {colorUseRule=UseColor_Specified, specifiedColor=<rgba 0 0 0 0>}
      cornerRadii=<0 0 0 0>
     }
     tintColor {colorUseRule=UseColor_Specified, specifiedColor=<rgba 1 1 1 1>}
     uVRegion {bIsValid=false, max=<0 0>, min=<0 0>}
    }}
   slot {class=HorizontalBoxSlot, properties {padding=<ltrb 15 0 15 0>}}
  }
  Image_341 {
   widgetClassPath=&Image
   bIsVariable=true
   parent=HorizontalBox_89
   widget=Image_341
   properties {brush {
     bIsDynamicallyLoaded=false
     drawAs=Image
     imageType=NoImage
     mirroring=NoMirror
     resourceName=None
     resourceObject=&Circle
     tiling=NoTile
     imageSize=<128 128>
     margin=<ltrb 0 0 0 0>
     outlineSettings {
      bUseBrushTransparency=false
      roundingType=HalfHeightRadius
      width=0
      color {colorUseRule=UseColor_Specified, specifiedColor=<rgba 0 0 0 0>}
      cornerRadii=<0 0 0 0>
     }
     tintColor {colorUseRule=UseColor_Specified, specifiedColor=<rgba 1 1 1 1>}
     uVRegion {bIsValid=false, max=<0 0>, min=<0 0>}
    }}
   slot {class=HorizontalBoxSlot, properties {padding=<ltrb 15 0 15 0>}}
  }
  Border_0 {
   widgetClassPath=&Border
   parent=VerticalBox_0
   widget=Border_0
   properties {
    horizontalAlignment=HAlign_Center
    background {
     bIsDynamicallyLoaded=false
     drawAs=Image
     imageType=NoImage
     mirroring=NoMirror
     resourceName=None
     resourceObject=null
     tiling=NoTile
     imageSize=<32 32>
     margin=<ltrb 0 0 0 0>
     outlineSettings {
      bUseBrushTransparency=false
      roundingType=HalfHeightRadius
      width=0
      color {colorUseRule=UseColor_Specified, specifiedColor=<rgba 0 0 0 0>}
      cornerRadii=<0 0 0 0>
     }
     tintColor {colorUseRule=UseColor_Specified, specifiedColor=<rgba 0.0262 0 0.0097 1>}
     uVRegion {bIsValid=false, max=<0 0>, min=<0 0>}
    }
    brushColor=<rgba 1 1 1 0.25>
    padding=<ltrb 384 32 384 32>
   }
   slot {class=VerticalBoxSlot, properties {padding=<ltrb -256 180 -256 0>, size {sizeRule=Fill, value=1}}}
  }
  Scroll_Box {
   widgetClassPath=&ScrollBox
   bInherited=true
   parent=Border_0
   widget=Scroll_Box
   properties {
    clipping=Inherit
    orientation=Orient_Horizontal
    scrollWhenFocusChanges=AnimatedScroll
    scrollbarThickness=<12 12>
    widgetBarStyle {
     thickness=16
     draggedThumbImage {
      bIsDynamicallyLoaded=false
      drawAs=Box
      imageType=FullColor
      mirroring=NoMirror
      resourceName=../../../Engine/Content/Slate/Common/Scrollbar_Thumb.png
      resourceObject=/Game/Athena/UI/ReplayBrowser/M_Ui_Skewed_Vfor9slices_BG_Separator_Inst
      tiling=NoTile
      imageSize=<12 12>
      margin=<ltrb 0.25 0.25 0.25 0.25>
      outlineSettings {
       bUseBrushTransparency=false
       roundingType=HalfHeightRadius
       width=0
       color {colorUseRule=UseColor_Specified, specifiedColor=<rgba 0 0 0 0>}
       cornerRadii=<0 0 0 0>
      }
      tintColor {colorUseRule=UseColor_Specified, specifiedColor=<rgba 1 1 1 1>}
      uVRegion {bIsValid=false, max=<0 0>, min=<0 0>}
     }
     horizontalBackgroundImage {
      bIsDynamicallyLoaded=false
      drawAs=Image
      imageType=NoImage
      mirroring=NoMirror
      resourceName=None
      resourceObject=&M_Ui_Skewed_Vfor9slices_BG_Separator_Inst
      tiling=NoTile
      imageSize=<0 0>
      margin=<ltrb 0 0 0 0>
      outlineSettings {
       bUseBrushTransparency=false
       roundingType=HalfHeightRadius
       width=0
       color {colorUseRule=UseColor_Specified, specifiedColor=<rgba 0 0 0 0>}
       cornerRadii=<0 0 0 0>
      }
      tintColor {colorUseRule=UseColor_Specified, specifiedColor=<rgba 1 0 0.0395 0.4>}
      uVRegion {bIsValid=false, max=<0 0>, min=<0 0>}
     }
     horizontalBottomSlotImage {
      bIsDynamicallyLoaded=false
      drawAs=NoDrawType
      imageType=FullColor
      mirroring=NoMirror
      resourceName=../../../Engine/Content/Slate/Common/Scrollbar_Background_Horizontal.png
      resourceObject=null
      tiling=NoTile
      imageSize=<8 8>
      margin=<ltrb 0 0 0 0>
      outlineSettings {
       bUseBrushTransparency=false
       roundingType=HalfHeightRadius
       width=0
       color {colorUseRule=UseColor_Specified, specifiedColor=<rgba 0 0 0 0>}
       cornerRadii=<0 0 0 0>
      }
      tintColor {colorUseRule=UseColor_Specified, specifiedColor=<rgba 1 1 1 1>}
      uVRegion {bIsValid=false, max=<0 0>, min=<0 0>}
     }
     horizontalTopSlotImage {
      bIsDynamicallyLoaded=false
      drawAs=NoDrawType
      imageType=FullColor
      mirroring=NoMirror
      resourceName=../../../Engine/Content/Slate/Common/Scrollbar_Background_Horizontal.png
      resourceObject=null
      tiling=NoTile
      imageSize=<8 8>
      margin=<ltrb 0 0 0 0>
      outlineSettings {
       bUseBrushTransparency=false
       roundingType=HalfHeightRadius
       width=0
       color {colorUseRule=UseColor_Specified, specifiedColor=<rgba 0 0 0 0>}
       cornerRadii=<0 0 0 0>
      }
      tintColor {colorUseRule=UseColor_Specified, specifiedColor=<rgba 1 1 1 1>}
      uVRegion {bIsValid=false, max=<0 0>, min=<0 0>}
     }
     hoveredThumbImage {
      bIsDynamicallyLoaded=false
      drawAs=Box
      imageType=FullColor
      mirroring=NoMirror
      resourceName=../../../Engine/Content/Slate/Common/Scrollbar_Thumb.png
      resourceObject=&M_Ui_Skewed_Vfor9slices_BG_Separator_Inst
      tiling=NoTile
      imageSize=<12 12>
      margin=<ltrb 0.25 0.25 0.25 0.25>
      outlineSettings {
       bUseBrushTransparency=false
       roundingType=HalfHeightRadius
       width=0
       color {colorUseRule=UseColor_Specified, specifiedColor=<rgba 0 0 0 0>}
       cornerRadii=<0 0 0 0>
      }
      tintColor {colorUseRule=UseColor_Specified, specifiedColor=<rgba 1 1 1 1>}
      uVRegion {bIsValid=false, max=<0 0>, min=<0 0>}
     }
     normalThumbImage {
      bIsDynamicallyLoaded=false
      drawAs=Box
      imageType=FullColor
      mirroring=NoMirror
      resourceName=../../../Engine/Content/Slate/Common/Scrollbar_Thumb.png
      resourceObject=&M_Ui_Skewed_Vfor9slices_BG_Separator_Inst
      tiling=NoTile
      imageSize=<12 12>
      margin=<ltrb 0.25 0.25 0.25 0.25>
      outlineSettings {
       bUseBrushTransparency=false
       roundingType=HalfHeightRadius
       width=0
       color {colorUseRule=UseColor_Specified, specifiedColor=<rgba 0 0 0 0>}
       cornerRadii=<0 0 0 0>
      }
      tintColor {colorUseRule=UseColor_Specified, specifiedColor=<rgba 1 1 1 1>}
      uVRegion {bIsValid=false, max=<0 0>, min=<0 0>}
     }
     verticalBackgroundImage {
      bIsDynamicallyLoaded=false
      drawAs=NoDrawType
      imageType=NoImage
      mirroring=NoMirror
      resourceName=None
      resourceObject=null
      tiling=NoTile
      imageSize=<0 0>
      margin=<ltrb 0 0 0 0>
      outlineSettings {
       bUseBrushTransparency=false
       roundingType=HalfHeightRadius
       width=0
       color {colorUseRule=UseColor_Specified, specifiedColor=<rgba 0 0 0 0>}
       cornerRadii=<0 0 0 0>
      }
      tintColor {colorUseRule=UseColor_Specified, specifiedColor=<rgba 1 1 1 1>}
      uVRegion {bIsValid=false, max=<0 0>, min=<0 0>}
     }
     verticalBottomSlotImage {
      bIsDynamicallyLoaded=false
      drawAs=NoDrawType
      imageType=FullColor
      mirroring=NoMirror
      resourceName=../../../Engine/Content/Slate/Common/Scrollbar_Background_Vertical.png
      resourceObject=null
      tiling=NoTile
      imageSize=<8 8>
      margin=<ltrb 0 0 0 0>
      outlineSettings {
       bUseBrushTransparency=false
       roundingType=HalfHeightRadius
       width=0
       color {colorUseRule=UseColor_Specified, specifiedColor=<rgba 0 0 0 0>}
       cornerRadii=<0 0 0 0>
      }
      tintColor {colorUseRule=UseColor_Specified, specifiedColor=<rgba 1 1 1 1>}
      uVRegion {bIsValid=false, max=<0 0>, min=<0 0>}
     }
     verticalTopSlotImage {
      bIsDynamicallyLoaded=false
      drawAs=NoDrawType
      imageType=FullColor
      mirroring=NoMirror
      resourceName=../../../Engine/Content/Slate/Common/Scrollbar_Background_Vertical.png
      resourceObject=null
      tiling=NoTile
      imageSize=<8 8>
      margin=<ltrb 0 0 0 0>
      outlineSettings {
       bUseBrushTransparency=false
       roundingType=HalfHeightRadius
       width=0
       color {colorUseRule=UseColor_Specified, specifiedColor=<rgba 0 0 0 0>}
       cornerRadii=<0 0 0 0>
      }
      tintColor {colorUseRule=UseColor_Specified, specifiedColor=<rgba 1 1 1 1>}
      uVRegion {bIsValid=false, max=<0 0>, min=<0 0>}
     }
    }
    widgetStyle {
     barThickness=9
     bottomShadowBrush {
      bIsDynamicallyLoaded=false
      drawAs=NoDrawType
      imageType=FullColor
      mirroring=NoMirror
      resourceName=../../../Engine/Content/Slate/Common/ScrollBoxShadowBottom.png
      resourceObject=null
      tiling=NoTile
      imageSize=<16 8>
      margin=<ltrb 0.5 0 0.5 1>
      outlineSettings {
       bUseBrushTransparency=false
       roundingType=HalfHeightRadius
       width=0
       color {colorUseRule=UseColor_Specified, specifiedColor=<rgba 0 0 0 0>}
       cornerRadii=<0 0 0 0>
      }
      tintColor {colorUseRule=UseColor_Specified, specifiedColor=<rgba 1 1 1 1>}
      uVRegion {bIsValid=false, max=<0 0>, min=<0 0>}
     }
     horizontalScrolledContentPadding=<ltrb 0 0 1 0>
     leftShadowBrush {
      bIsDynamicallyLoaded=false
      drawAs=NoDrawType
      imageType=FullColor
      mirroring=NoMirror
      resourceName=../../../Engine/Content/Slate/Common/ScrollBoxShadowLeft.png
      resourceObject=null
      tiling=NoTile
      imageSize=<8 16>
      margin=<ltrb 1 0.5 0 0.5>
      outlineSettings {
       bUseBrushTransparency=false
       roundingType=HalfHeightRadius
       width=0
       color {colorUseRule=UseColor_Specified, specifiedColor=<rgba 0 0 0 0>}
       cornerRadii=<0 0 0 0>
      }
      tintColor {colorUseRule=UseColor_Specified, specifiedColor=<rgba 1 1 1 1>}
      uVRegion {bIsValid=false, max=<0 0>, min=<0 0>}
     }
     rightShadowBrush {
      bIsDynamicallyLoaded=false
      drawAs=NoDrawType
      imageType=FullColor
      mirroring=NoMirror
      resourceName=../../../Engine/Content/Slate/Common/ScrollBoxShadowRight.png
      resourceObject=null
      tiling=NoTile
      imageSize=<8 16>
      margin=<ltrb 0 0.5 1 0.5>
      outlineSettings {
       bUseBrushTransparency=false
       roundingType=HalfHeightRadius
       width=0
       color {colorUseRule=UseColor_Specified, specifiedColor=<rgba 0 0 0 0>}
       cornerRadii=<0 0 0 0>
      }
      tintColor {colorUseRule=UseColor_Specified, specifiedColor=<rgba 1 1 1 1>}
      uVRegion {bIsValid=false, max=<0 0>, min=<0 0>}
     }
     topShadowBrush {
      bIsDynamicallyLoaded=false
      drawAs=NoDrawType
      imageType=FullColor
      mirroring=NoMirror
      resourceName=../../../Engine/Content/Slate/Common/ScrollBoxShadowTop.png
      resourceObject=null
      tiling=NoTile
      imageSize=<16 8>
      margin=<ltrb 0.5 1 0.5 0>
      outlineSettings {
       bUseBrushTransparency=false
       roundingType=HalfHeightRadius
       width=0
       color {colorUseRule=UseColor_Specified, specifiedColor=<rgba 0 0 0 0>}
       cornerRadii=<0 0 0 0>
      }
      tintColor {colorUseRule=UseColor_Specified, specifiedColor=<rgba 1 1 1 1>}
      uVRegion {bIsValid=false, max=<0 0>, min=<0 0>}
     }
     verticalScrolledContentPadding=<ltrb 0 0 0 1>
    }
   }
   slot {class=BorderSlot, properties {horizontalAlignment=HAlign_Center, padding=<ltrb 384 32 384 32>}}
  }
  EntryBox_PunchCards {
   widgetClassPath=&DynamicEntryBox
   bInherited=true
   bIsVariable=true
   parent=Scroll_Box
   widget=EntryBox_PunchCards
   properties {
    entryHorizontalAlignment=HAlign_Center
    entryWidgetClass=&/SpatialChallengePunchCardEntry.SpatialChallengePunchCardEntry_C
    numDesignerPreviewEntries=4
    entrySizeRule {sizeRule=Fill, value=1}
    entrySpacing=<20 0>
   }
   slot {class=ScrollBoxSlot, properties {verticalAlignment=VAlign_Center, padding=<ltrb 0 0 0 8>}}
  }
  HorizontalBox_0 {
   widgetClassPath=&HorizontalBox
   parent=VerticalBox_0
   widget=HorizontalBox_0
   slot {class=VerticalBoxSlot, properties {verticalAlignment=VAlign_Bottom, padding=<ltrb 32 32 32 32>}}
  }
  SizeBox_172 {
   widgetClassPath=&SizeBox
   parent=HorizontalBox_0
   widget=SizeBox_172
   properties {bOverride_HeightOverride=true, heightOverride=55}
   slot {class=HorizontalBoxSlot, properties {horizontalAlignment=HAlign_Right, size {sizeRule=Fill, value=1}}}
  }
  HorizontalBox_1 {
   widgetClassPath=&HorizontalBox
   parent=SizeBox_172
   widget=HorizontalBox_1
   slot {class=SizeBoxSlot, properties {horizontalAlignment=HAlign_Right, verticalAlignment=VAlign_Bottom}}
  }
  Button_About {
   widgetClassPath=&IconTextButton
   bIsVariable=true
   parent=HorizontalBox_1
   widget=Button_About
   properties {
    buttonText=More Info
    minWidth=250
    onButtonBaseClicked=(&self:WidgetTree.Button_About.HandleOnButtonClicked)
    onButtonBaseDoubleClicked=(&self:WidgetTree.Button_About.HandleOnButtonDoubleClicked)
    onButtonBaseHovered=(&self:WidgetTree.Button_About.HandleOnButtonHovered)
    onButtonBaseUnhovered=(&self:WidgetTree.Button_About.HandleOnButtonUnhovered)
    onSelectedChangedBase=(&self:WidgetTree.Button_About.HandleOnSelectedChanged)
    style=/BRCosmetics/UI/Foundation/Buttons/ButtonStyles/Athena/Athena_ButtonStyle_Hightower_Yellow.Athena_ButtonStyle_Hightower_Yellow_C
    triggeringInputAction {dataTable=&InputActionDataTable, rowName=GenericFaceButtonTop}
   }
   slot {class=HorizontalBoxSlot, properties {verticalAlignment=VAlign_Bottom, padding=<ltrb 0 0 25 0>}}
  }
  HardwareVisibilityBorder_Back {
   widgetClassPath=&CommonHardwareVisibilityBorder
   parent=HorizontalBox_1
   widget=HardwareVisibilityBorder_Back
   properties {visibilityQuery {
     autoDescription=" NONE( Input.Touch )"
     tokenStreamVersion=0
     userDescription=""
     queryTokenStream [0,1,3,1,0]
     tagDictionary [{tagName=Input.Touch}]
    }}
   slot {class=HorizontalBoxSlot}
  }
  Button_Back {
   widgetClassPath=&IconTextButton
   bInherited=true
   bIsVariable=true
   parent=HardwareVisibilityBorder_Back
   widget=Button_Back
   properties {
    buttonText=BACK
    minHeight=13
    minWidth=200
    onButtonBaseClicked=(&self:WidgetTree.Button_Back.HandleOnButtonClicked)
    onButtonBaseDoubleClicked=(&self:WidgetTree.Button_Back.HandleOnButtonDoubleClicked)
    onButtonBaseHovered=(&self:WidgetTree.Button_Back.HandleOnButtonHovered)
    onButtonBaseUnhovered=(&self:WidgetTree.Button_Back.HandleOnButtonUnhovered)
    onSelectedChangedBase=(&self:WidgetTree.Button_Back.HandleOnSelectedChanged)
    style=/BRCosmetics/UI/Foundation/Buttons/ButtonStyles/Athena/Athena_ButtonStyle_Hightower_red.Athena_ButtonStyle_Hightower_red_C
    triggeringInputAction {dataTable=&InputActionDataTable, rowName=Back}
   }
   slot {class=BorderSlot, properties {padding=<ltrb 0 0 0 0>}}
  }
  BPS12_AboutModal {
   widgetClassPath=&/BPS12_AboutModal
   bIsVariable=true
   parent=CommonActivatableWidgetSwitcher_State
   widget=BPS12_AboutModal
   properties {
    header=PUNCH CARDS
    header_Icon=/BRCosmetics/Athena/UI/Challenges/Art/M_Ui_PuncardIcon_Inst_Inst
    show Header=false
    subHeader="NEW FOR SEASON 3!"
    title="TRACK PROGRESS && EARN XP!"
    divider_Color=<rgba 1 0 0.0507 1>
    entries [
     "Experiment on the island to <text color=\"FFFFFFFF\" fontface=\"black\" case=\"upper\">discover punch cards</>."
     "<text color=\"FFFFFFFF\" fontface=\"black\" case=\"upper\">Reach Milestones</> in each punch card to <text color=\"FFFFFFFF\" fontface=\"black\" case=\"upper\">Earn XP</>."
    ]
    subheader_Color=<rgba 1 0.5043 0 1>
   }
   slot {class=WidgetSwitcherSlot}
  }
  Overlay_InfoCards {
   widgetClassPath=&Overlay
   bIsVariable=true
   parent=Overlay_0
   widget=Overlay_InfoCards
   properties {renderOpacity=0, visibility=HitTestInvisible}
   slot {class=OverlaySlot, properties {horizontalAlignment=HAlign_Right, verticalAlignment=VAlign_Center}}
  }
  SizeBox_Card2 {
   widgetClassPath=&SizeBox
   parent=Overlay_InfoCards
   widget=SizeBox_Card2
   properties {
    bOverride_WidthOverride=true
    heightOverride=308.3217
    widthOverride=525
    renderTransform {angle=13, scale=<1 1>, shear=<0 0>, translation=<7 13>}
   }
   slot {class=OverlaySlot, properties {horizontalAlignment=HAlign_Fill, verticalAlignment=VAlign_Center, padding=<ltrb 150 0 150 0>}}
  }
  Overlay_Card1_1 {widgetClassPath=&Overlay, parent=SizeBox_Card2, widget=Overlay_Card1_1, slot {class=SizeBoxSlot}}
  Image_shadow_1 {
   widgetClassPath=&Image
   bIsVariable=true
   parent=Overlay_Card1_1
   widget=Image_shadow_1
   properties {
    renderOpacity=0.25
    brush {
     bIsDynamicallyLoaded=false
     drawAs=Box
     imageType=NoImage
     mirroring=NoMirror
     resourceName=None
     resourceObject=/Game/Athena/UI/Challenges/Art/M_Ui_Punchcard_Inst_Inst_Inst1
     tiling=NoTile
     imageSize=<256 256>
     margin=<ltrb 0.25 0.25 0.25 0.25>
     outlineSettings {
      bUseBrushTransparency=false
      roundingType=HalfHeightRadius
      width=0
      color {colorUseRule=UseColor_Specified, specifiedColor=<rgba 0 0 0 0>}
      cornerRadii=<0 0 0 0>
     }
     tintColor {colorUseRule=UseColor_Specified, specifiedColor=<rgba 0 0 0 1>}
     uVRegion {bIsValid=false, max=<0 0>, min=<0 0>}
    }
    renderTransform {angle=0, scale=<1.01 1.02>, shear=<0 0>, translation=<0 0>}
   }
   slot {class=OverlaySlot, properties {horizontalAlignment=HAlign_Fill, verticalAlignment=VAlign_Fill}}
  }
  FortLazyImage_Background_02 {
   widgetClassPath=&FortLazyImage
   bIsVariable=true
   parent=Overlay_Card1_1
   widget=FortLazyImage_Background_02
   properties {brush {
     bIsDynamicallyLoaded=false
     drawAs=Box
     imageType=NoImage
     mirroring=NoMirror
     resourceName=None
     resourceObject=/Game/Athena/UI/Challenges/Art/M_Ui_Punchcard_Inst
     tiling=NoTile
     imageSize=<128 128>
     margin=<ltrb 0.2 0.2 0.2 0.2>
     outlineSettings {
      bUseBrushTransparency=false
      roundingType=HalfHeightRadius
      width=0
      color {colorUseRule=UseColor_Specified, specifiedColor=<rgba 0 0 0 0>}
      cornerRadii=<0 0 0 0>
     }
     tintColor {colorUseRule=UseColor_Specified, specifiedColor=<rgba 1 1 1 1>}
     uVRegion {bIsValid=false, max=<0 0>, min=<0 0>}
    }}
   slot {class=OverlaySlot, properties {horizontalAlignment=HAlign_Fill, verticalAlignment=VAlign_Fill}}
  }
  VerticalBox {
   widgetClassPath=&VerticalBox
   parent=Overlay_Card1_1
   widget=VerticalBox
   slot {class=OverlaySlot, properties {horizontalAlignment=HAlign_Fill, padding=<ltrb 20 6 20 6>}}
  }
  SizeBox_1 {
   widgetClassPath=&SizeBox
   parent=VerticalBox
   widget=SizeBox_1
   properties {bOverride_HeightOverride=true, heightOverride=116}
   slot {class=VerticalBoxSlot}
  }
  CommonTextBlock {
   widgetClassPath=&CommonTextBlock
   parent=SizeBox_1
   widget=CommonTextBlock
   properties {
    justification=Center
    scrollStyle=&TextScrollStyle-Base_C
    text=Double-Dipping
    textTransformPolicy=ToUpper
    font {
     bForceMonospaced=false
     bMaterialIsStencil=false
     fontMaterial=null
     fontObject=&Burbank
     letterSpacing=0
     monospacedWidth=1
     size=26
     skewAmount=0
     typefaceFontName=Burbank-PlayerPower
     outlineSettings {
      bApplyOutlineToDropShadows=false
      bMiteredCorners=false
      bSeparateFillAlpha=false
      outlineMaterial=null
      outlineSize=0
      outlineColor=<rgba 0 0 0 1>
     }
    }
    renderTransform {angle=0, scale=<1 1>, shear=<-12 0>, translation=<0 0>}
    shadowOffset=<0 0>
   }
   slot {class=SizeBoxSlot, properties {verticalAlignment=VAlign_Center}}
  }
  SizeBox_2 {
   widgetClassPath=&SizeBox
   parent=VerticalBox
   widget=SizeBox_2
   properties {bOverride_HeightOverride=true, heightOverride=140}
   slot {class=VerticalBoxSlot}
  }
  ScaleBox {widgetClassPath=&ScaleBox, parent=SizeBox_2, widget=ScaleBox, slot {class=SizeBoxSlot}}
  HorizontalBox {
   widgetClassPath=&HorizontalBox
   parent=ScaleBox
   widget=HorizontalBox
   properties {renderTransform {angle=0, scale=<1 1>, shear=<0 0>, translation=<0 -6>}}
   slot {class=ScaleBoxSlot}
  }
  Image_EmptySocket_02_left {
   widgetClassPath=&Image
   bIsVariable=true
   parent=HorizontalBox
   widget=Image_EmptySocket_02_left
   properties {
    brush {
     bIsDynamicallyLoaded=false
     drawAs=Image
     imageType=NoImage
     mirroring=NoMirror
     resourceName=None
     resourceObject=/Game/Athena/UI/Challenges/Art/M_Ui_Punchcard_slot_Inst
     tiling=NoTile
     imageSize=<110 110>
     margin=<ltrb 0 0 0 0>
     outlineSettings {
      bUseBrushTransparency=false
      roundingType=HalfHeightRadius
      width=0
      color {colorUseRule=UseColor_Specified, specifiedColor=<rgba 0 0 0 0>}
      cornerRadii=<0 0 0 0>
     }
     tintColor {colorUseRule=UseColor_Specified, specifiedColor=<rgba 1 1 1 0.5>}
     uVRegion {bIsValid=false, max=<0 0>, min=<0 0>}
    }
    colorAndOpacity=<rgba 0.0262 0 0.0097 1>
   }
   slot {class=HorizontalBoxSlot}
  }
  Image_EmptySocket {
   widgetClassPath=&Image
   parent=HorizontalBox
   widget=Image_EmptySocket
   properties {
    brush {
     bIsDynamicallyLoaded=false
     drawAs=Image
     imageType=NoImage
     mirroring=NoMirror
     resourceName=None
     resourceObject=&M_Ui_Punchcard_slot_Inst
     tiling=NoTile
     imageSize=<110 110>
     margin=<ltrb 0 0 0 0>
     outlineSettings {
      bUseBrushTransparency=false
      roundingType=HalfHeightRadius
      width=0
      color {colorUseRule=UseColor_Specified, specifiedColor=<rgba 0 0 0 0>}
      cornerRadii=<0 0 0 0>
     }
     tintColor {colorUseRule=UseColor_Specified, specifiedColor=<rgba 1 1 1 0.5>}
     uVRegion {bIsValid=false, max=<0 0>, min=<0 0>}
    }
    colorAndOpacity=<rgba 0.0262 0 0.0097 1>
   }
   slot {class=HorizontalBoxSlot}
  }
  Image_EmptySocket_5 {
   widgetClassPath=&Image
   parent=HorizontalBox
   widget=Image_EmptySocket_5
   properties {
    brush {
     bIsDynamicallyLoaded=false
     drawAs=Image
     imageType=NoImage
     mirroring=NoMirror
     resourceName=None
     resourceObject=&M_Ui_Punchcard_slot_Inst
     tiling=NoTile
     imageSize=<110 110>
     margin=<ltrb 0 0 0 0>
     outlineSettings {
      bUseBrushTransparency=false
      roundingType=HalfHeightRadius
      width=0
      color {colorUseRule=UseColor_Specified, specifiedColor=<rgba 0 0 0 0>}
      cornerRadii=<0 0 0 0>
     }
     tintColor {colorUseRule=UseColor_Specified, specifiedColor=<rgba 1 1 1 0.5>}
     uVRegion {bIsValid=false, max=<0 0>, min=<0 0>}
    }
    colorAndOpacity=<rgba 0.0262 0 0.0097 1>
   }
   slot {class=HorizontalBoxSlot}
  }
  Image_EmptySocket_02_Right {
   widgetClassPath=&Image
   bIsVariable=true
   parent=HorizontalBox
   widget=Image_EmptySocket_02_Right
   properties {
    brush {
     bIsDynamicallyLoaded=false
     drawAs=Image
     imageType=NoImage
     mirroring=NoMirror
     resourceName=None
     resourceObject=&M_Ui_Punchcard_slot_Inst
     tiling=NoTile
     imageSize=<110 110>
     margin=<ltrb 0 0 0 0>
     outlineSettings {
      bUseBrushTransparency=false
      roundingType=HalfHeightRadius
      width=0
      color {colorUseRule=UseColor_Specified, specifiedColor=<rgba 0 0 0 0>}
      cornerRadii=<0 0 0 0>
     }
     tintColor {colorUseRule=UseColor_Specified, specifiedColor=<rgba 1 1 1 0.5>}
     uVRegion {bIsValid=false, max=<0 0>, min=<0 0>}
    }
    colorAndOpacity=<rgba 0.0262 0 0.0097 1>
   }
   slot {class=HorizontalBoxSlot}
  }
  SizeBox_Card1 {
   widgetClassPath=&SizeBox
   parent=Overlay_InfoCards
   widget=SizeBox_Card1
   properties {
    bOverride_WidthOverride=true
    heightOverride=308.3217
    widthOverride=525
    renderTransform {angle=-13, scale=<1 1>, shear=<0 0>, translation=<-79 -166>}
   }
   slot {class=OverlaySlot, properties {horizontalAlignment=HAlign_Fill, verticalAlignment=VAlign_Center, padding=<ltrb 150 0 150 0>}}
  }
  Overlay_Card1 {widgetClassPath=&Overlay, parent=SizeBox_Card1, widget=Overlay_Card1, slot {class=SizeBoxSlot}}
  Image_shadow {
   widgetClassPath=&Image
   bIsVariable=true
   parent=Overlay_Card1
   widget=Image_shadow
   properties {
    renderOpacity=0.25
    brush {
     bIsDynamicallyLoaded=false
     drawAs=Box
     imageType=NoImage
     mirroring=NoMirror
     resourceName=None
     resourceObject=&M_Ui_Punchcard_Inst_Inst_Inst1
     tiling=NoTile
     imageSize=<256 256>
     margin=<ltrb 0.25 0.25 0.25 0.25>
     outlineSettings {
      bUseBrushTransparency=false
      roundingType=HalfHeightRadius
      width=0
      color {colorUseRule=UseColor_Specified, specifiedColor=<rgba 0 0 0 0>}
      cornerRadii=<0 0 0 0>
     }
     tintColor {colorUseRule=UseColor_Specified, specifiedColor=<rgba 0 0 0 1>}
     uVRegion {bIsValid=false, max=<0 0>, min=<0 0>}
    }
    renderTransform {angle=0, scale=<1.015 1.015>, shear=<2 0>, translation=<6 5>}
   }
   slot {class=OverlaySlot, properties {horizontalAlignment=HAlign_Fill, verticalAlignment=VAlign_Fill}}
  }
  FortLazyImage_Background_01 {
   widgetClassPath=&FortLazyImage
   bIsVariable=true
   parent=Overlay_Card1
   widget=FortLazyImage_Background_01
   properties {brush {
     bIsDynamicallyLoaded=false
     drawAs=Box
     imageType=NoImage
     mirroring=NoMirror
     resourceName=None
     resourceObject=&M_Ui_Punchcard_Inst
     tiling=NoTile
     imageSize=<128 128>
     margin=<ltrb 0.2 0.2 0.2 0.2>
     outlineSettings {
      bUseBrushTransparency=false
      roundingType=HalfHeightRadius
      width=0
      color {colorUseRule=UseColor_Specified, specifiedColor=<rgba 0 0 0 0>}
      cornerRadii=<0 0 0 0>
     }
     tintColor {colorUseRule=UseColor_Specified, specifiedColor=<rgba 1 1 1 1>}
     uVRegion {bIsValid=false, max=<0 0>, min=<0 0>}
    }}
   slot {class=OverlaySlot, properties {horizontalAlignment=HAlign_Fill, verticalAlignment=VAlign_Fill}}
  }
  VerticalBox_2 {
   widgetClassPath=&VerticalBox
   parent=Overlay_Card1
   widget=VerticalBox_2
   slot {class=OverlaySlot, properties {horizontalAlignment=HAlign_Fill, padding=<ltrb 20 6 20 6>}}
  }
  SizeBox_140 {
   widgetClassPath=&SizeBox
   parent=VerticalBox_2
   widget=SizeBox_140
   properties {bOverride_HeightOverride=true, heightOverride=116}
   slot {class=VerticalBoxSlot}
  }
  CommonTextBlock_81 {
   widgetClassPath=&CommonTextBlock
   parent=SizeBox_140
   widget=CommonTextBlock_81
   properties {
    justification=Center
    scrollStyle=&TextScrollStyle-Base_C
    text=I CALL SHOTGUN
    textTransformPolicy=ToUpper
    font {
     bForceMonospaced=false
     bMaterialIsStencil=false
     fontMaterial=null
     fontObject=&Burbank
     letterSpacing=0
     monospacedWidth=1
     size=26
     skewAmount=0
     typefaceFontName=Burbank-PlayerPower
     outlineSettings {
      bApplyOutlineToDropShadows=false
      bMiteredCorners=false
      bSeparateFillAlpha=false
      outlineMaterial=null
      outlineSize=0
      outlineColor=<rgba 0 0 0 1>
     }
    }
    renderTransform {angle=0, scale=<1 1>, shear=<-12 0>, translation=<0 0>}
    shadowOffset=<0 0>
   }
   slot {class=SizeBoxSlot, properties {verticalAlignment=VAlign_Center}}
  }
  SizeBox {
   widgetClassPath=&SizeBox
   parent=VerticalBox_2
   widget=SizeBox
   properties {bOverride_HeightOverride=true, heightOverride=140}
   slot {class=VerticalBoxSlot}
  }
  ScaleBox_0 {widgetClassPath=&ScaleBox, parent=SizeBox, widget=ScaleBox_0, slot {class=SizeBoxSlot}}
  HorizontalBox_2 {
   widgetClassPath=&HorizontalBox
   parent=ScaleBox_0
   widget=HorizontalBox_2
   properties {renderTransform {angle=0, scale=<1 1>, shear=<0 0>, translation=<0 -14>}}
   slot {class=ScaleBoxSlot}
  }
  Overlay_192 {widgetClassPath=&Overlay, parent=HorizontalBox_2, widget=Overlay_192, slot {class=HorizontalBoxSlot}}
  Image_EmptySocket_01_left {
   widgetClassPath=&Image
   bIsVariable=true
   parent=Overlay_192
   widget=Image_EmptySocket_01_left
   properties {
    brush {
     bIsDynamicallyLoaded=false
     drawAs=Image
     imageType=NoImage
     mirroring=NoMirror
     resourceName=None
     resourceObject=&M_Ui_Punchcard_slot_Inst
     tiling=NoTile
     imageSize=<110 110>
     margin=<ltrb 0 0 0 0>
     outlineSettings {
      bUseBrushTransparency=false
      roundingType=HalfHeightRadius
      width=0
      color {colorUseRule=UseColor_Specified, specifiedColor=<rgba 0 0 0 0>}
      cornerRadii=<0 0 0 0>
     }
     tintColor {colorUseRule=UseColor_Specified, specifiedColor=<rgba 1 1 1 0.5>}
     uVRegion {bIsValid=false, max=<0 0>, min=<0 0>}
    }
    colorAndOpacity=<rgba 0.0262 0 0.0097 1>
   }
   slot {class=OverlaySlot}
  }
  Image_217 {
   widgetClassPath=&Image
   bIsVariable=true
   parent=Overlay_192
   widget=Image_217
   properties {
    brush {
     bIsDynamicallyLoaded=false
     drawAs=Image
     imageType=NoImage
     mirroring=NoMirror
     resourceName=None
     resourceObject=/BRCosmetics/Athena/UI/Challenges/Art/M_Ui_PuncardIcon_Inst_Inst1
     tiling=NoTile
     imageSize=<32 32>
     margin=<ltrb 0 0 0 0>
     outlineSettings {
      bUseBrushTransparency=false
      roundingType=HalfHeightRadius
      width=0
      color {colorUseRule=UseColor_Specified, specifiedColor=<rgba 0 0 0 0>}
      cornerRadii=<0 0 0 0>
     }
     tintColor {colorUseRule=UseColor_Specified, specifiedColor=<rgba 1 1 1 1>}
     uVRegion {bIsValid=false, max=<0 0>, min=<0 0>}
    }
    renderTransform {angle=-10, scale=<1.015 1.015>, shear=<0 0>, translation=<1 3>}
   }
   slot {class=OverlaySlot, properties {horizontalAlignment=HAlign_Fill, verticalAlignment=VAlign_Fill}}
  }
  Image_EmptySocket_1 {
   widgetClassPath=&Image
   parent=HorizontalBox_2
   widget=Image_EmptySocket_1
   properties {
    brush {
     bIsDynamicallyLoaded=false
     drawAs=Image
     imageType=NoImage
     mirroring=NoMirror
     resourceName=None
     resourceObject=&M_Ui_Punchcard_slot_Inst
     tiling=NoTile
     imageSize=<110 110>
     margin=<ltrb 0 0 0 0>
     outlineSettings {
      bUseBrushTransparency=false
      roundingType=HalfHeightRadius
      width=0
      color {colorUseRule=UseColor_Specified, specifiedColor=<rgba 0 0 0 0>}
      cornerRadii=<0 0 0 0>
     }
     tintColor {colorUseRule=UseColor_Specified, specifiedColor=<rgba 1 1 1 0.5>}
     uVRegion {bIsValid=false, max=<0 0>, min=<0 0>}
    }
    colorAndOpacity=<rgba 0.0262 0 0.0097 1>
   }
   slot {class=HorizontalBoxSlot}
  }
  Image_EmptySocket_2 {
   widgetClassPath=&Image
   parent=HorizontalBox_2
   widget=Image_EmptySocket_2
   properties {
    brush {
     bIsDynamicallyLoaded=false
     drawAs=Image
     imageType=NoImage
     mirroring=NoMirror
     resourceName=None
     resourceObject=&M_Ui_Punchcard_slot_Inst
     tiling=NoTile
     imageSize=<110 110>
     margin=<ltrb 0 0 0 0>
     outlineSettings {
      bUseBrushTransparency=false
      roundingType=HalfHeightRadius
      width=0
      color {colorUseRule=UseColor_Specified, specifiedColor=<rgba 0 0 0 0>}
      cornerRadii=<0 0 0 0>
     }
     tintColor {colorUseRule=UseColor_Specified, specifiedColor=<rgba 1 1 1 0.5>}
     uVRegion {bIsValid=false, max=<0 0>, min=<0 0>}
    }
    colorAndOpacity=<rgba 0.0262 0 0.0097 1>
   }
   slot {class=HorizontalBoxSlot}
  }
  Image_EmptySocket_3 {
   widgetClassPath=&Image
   parent=HorizontalBox_2
   widget=Image_EmptySocket_3
   properties {
    brush {
     bIsDynamicallyLoaded=false
     drawAs=Image
     imageType=NoImage
     mirroring=NoMirror
     resourceName=None
     resourceObject=&M_Ui_Punchcard_slot_Inst
     tiling=NoTile
     imageSize=<110 110>
     margin=<ltrb 0 0 0 0>
     outlineSettings {
      bUseBrushTransparency=false
      roundingType=HalfHeightRadius
      width=0
      color {colorUseRule=UseColor_Specified, specifiedColor=<rgba 0 0 0 0>}
      cornerRadii=<0 0 0 0>
     }
     tintColor {colorUseRule=UseColor_Specified, specifiedColor=<rgba 1 1 1 0.5>}
     uVRegion {bIsValid=false, max=<0 0>, min=<0 0>}
    }
    colorAndOpacity=<rgba 0.0262 0 0.0097 1>
   }
   slot {class=HorizontalBoxSlot}
  }
  Image_EmptySocket_4 {
   widgetClassPath=&Image
   parent=HorizontalBox_2
   widget=Image_EmptySocket_4
   properties {
    brush {
     bIsDynamicallyLoaded=false
     drawAs=Image
     imageType=NoImage
     mirroring=NoMirror
     resourceName=None
     resourceObject=&M_Ui_Punchcard_slot_Inst
     tiling=NoTile
     imageSize=<110 110>
     margin=<ltrb 0 0 0 0>
     outlineSettings {
      bUseBrushTransparency=false
      roundingType=HalfHeightRadius
      width=0
      color {colorUseRule=UseColor_Specified, specifiedColor=<rgba 0 0 0 0>}
      cornerRadii=<0 0 0 0>
     }
     tintColor {colorUseRule=UseColor_Specified, specifiedColor=<rgba 1 1 1 0.5>}
     uVRegion {bIsValid=false, max=<0 0>, min=<0 0>}
    }
    colorAndOpacity=<rgba 0.0262 0 0.0097 1>
   }
   slot {class=HorizontalBoxSlot}
  }
  Image_EmptySocket_01_Right {
   widgetClassPath=&Image
   bIsVariable=true
   parent=HorizontalBox_2
   widget=Image_EmptySocket_01_Right
   properties {
    brush {
     bIsDynamicallyLoaded=false
     drawAs=Image
     imageType=NoImage
     mirroring=NoMirror
     resourceName=None
     resourceObject=&M_Ui_Punchcard_slot_Inst
     tiling=NoTile
     imageSize=<110 110>
     margin=<ltrb 0 0 0 0>
     outlineSettings {
      bUseBrushTransparency=false
      roundingType=HalfHeightRadius
      width=0
      color {colorUseRule=UseColor_Specified, specifiedColor=<rgba 0 0 0 0>}
      cornerRadii=<0 0 0 0>
     }
     tintColor {colorUseRule=UseColor_Specified, specifiedColor=<rgba 1 1 1 0.5>}
     uVRegion {bIsValid=false, max=<0 0>, min=<0 0>}
    }
    colorAndOpacity=<rgba 0.0262 0 0.0097 1>
   }
   slot {class=HorizontalBoxSlot}
  }
 ]
}
