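# SpatialMainRoomScreen is a Widget Blueprint subclass of AthenaSpatialScreen, apparently the main room screen of the spatial frontend UI.
# Its layout is a root Overlay holding one empty BackgroundBlur anchored bottom-center.
# The inherited Button_Back and Button_TouchBack slots have no widgets in this tree.
# The event graph has only unconnected PreConstruct, Construct and Tick events, so any behavior comes from the parent class.
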WidgetBlueprint /BRCosmetics/Athena/UI/Frontend/Spatial/SpatialMainRoomScreen
graphs [EventGraph {nodes [
   Event PreConstruct()
   Event Construct()
   Event Tick()
  ]}]
variables []
widgets {
 info {inheritedWidgetCount=2, parentClass=/Script/FortniteUI.AthenaSpatialScreen, rootWidgetClass=&Overlay, widgetCount=4}
 widgets [
  Overlay_0 {widgetClassPath=&Overlay, widget=Overlay_0}
  BackgroundBlur_0 {
   widgetClassPath=&BackgroundBlur
   parent=Overlay_0
   widget=BackgroundBlur_0
   properties {horizontalAlignment=HAlign_Center, verticalAlignment=VAlign_Bottom, padding=<ltrb 60 60 60 60>}
   slot {class=OverlaySlot, properties {horizontalAlignment=HAlign_Center, verticalAlignment=VAlign_Bottom}}
  }
  Button_Back {widgetClassPath=&CommonButtonLegacy, bInherited=true, widget=null}
  Button_TouchBack {widgetClassPath=&CommonButtonLegacy, bInherited=true, widget=null}
 ]
}
